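# Translucent, unlit, two-sided effect material for the narrow skirt of the Buffet Shard female character, usable on skeletal meshes and Niagara particles.
# The skirt adds panning wave and channel-offset fire noise layers over a corona gradient, scaled by time-of-day color.
# Its opacity combines scrolling rays, smoke, UV edge masks, view-angle and depth fades.
# The Only_use_the_head_halo switch instead selects a rotating, flickering spike-flare halo driven by particle color and dynamic parameters.
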Material /BRCosmetics/Characters/Player/Female/Medium/Bodies/F_MED_Buffet_Shard/FX/M_F_MED_Buffet_Shard_Skirt_Narrow
graph {
 properties {
  bUsedWithMeshDeformer=true
  bUsedWithMorphTargets=true
  bUsedWithNiagaraMeshParticles=true
  bUsedWithNiagaraSprites=true
  bUsedWithSkeletalMesh=true
  blendMode=BLEND_Translucent
  pixelDepthOffsetMode=PDOM_Legacy
  shadingModel=MSM_Unlit
  twoSided=true
  usedShadingModels=Unlit
 }
 nodes [
  n1 = VectorParameter() {parameterName=Color_Morning, defaultValue=<rgba 0.2 0.2 0.2 1>}
  n2 = VectorParameter() {parameterName=Color_Evening, defaultValue=<rgba 0.3 0.3 0.3 1>}
  n3 = VectorParameter() {parameterName=Color_Night, defaultValue=<rgba 1 1 1 0.8>}
  n4 = MaterialFunctionCall(Color_Morning=@n1.RGB, Color_Day=Constant() {r=1}, Color_Evening=@n2.RGB, Color_Night=@n3.RGB) {materialFunction=&Color_Time_of_Day}
  n5 = Add(A=NamedRerouteUsage(), B=NamedRerouteUsage())
  n6 = Constant3Vector() {constant=<rgba 4.1667 4.8812 10 0>}
  n7 = Multiply(A=Desaturation(None=Multiply(A=@n5, B=@n6), Fraction=Constant() {r=-0.5})) {constB=0.1}
  n8 = ParticleRandom()
  n9 = Multiply(A=Time(), B=Add(A=Multiply(A=@n8) {constB=0.1}))
  n10 = Multiply(A=Sine(None=@n9) {period=0.155}, B=Sine(None=@n9) {period=0.2})
  n11 = Sine(None=@n9) {period=0.43}
  n12 = Clamp(None=Multiply(A=Sine(None=@n9) {period=0.075}, B=Multiply(A=@n10, B=@n11))) {maxDefault=3}
  n13 = DynamicParameter() {defaultValue=<rgba 0.6 1 1 1>, paramNames [Size,RotateSpeed,Flicker,ColorJitter]}
  n14 = ParticleColor()
  n15 = Multiply(A=Time(), B=@n13.RotateSpeed)
  n16 = Rotator(Coordinate=Add(A=TextureCoordinate() {uTiling=0.25, vTiling=0.25}) {constB=0.375}, Time=@n15)
  n17 = DynamicParameter() {desc=Niagara index 1, parameterIndex=1, defaultValue=<rgba 6.667 0 1 1>, paramNames [ErodeSize, Erode, DepthFadeMult, Param4]}
  n18 = MaterialFunctionCall(Float4=LinearInterpolate(
    A=Constant4Vector() {constant=<rgba 6.667 0 1 0>}.RGBA
    B=MaterialFunctionCall(R=@n17.ErodeSize, G=@n17.ErodeSize, B=@n17.ErodeSize, A=@n17.ErodeSize) {materialFunction=&AppendMany}.RGBA
    Alpha=ScalarParameter() {parameterName=UseIndex1Params}
   )) {materialFunction=&BreakOutFloat4Components}
  n19 = MaterialFunctionCall(Texture Scale=@n18.R) {materialFunction=&ScaleUVsByCenter}
  n20 = OneMinus(None=TextureSample(UVs=Rotator(Coordinate=@n19.UVs, Time=@n15) {speed=-0.72}) {texture=/Game/Effects/Fort_Effects/Textures/Flares/T_Flare_Spikeball}.RGB)
  n21 = Multiply(
   A=Multiply(A=TextureSample(UVs=@n16) {texture=&T_Flare_Spikeball}.RGB, B=Power(Base=@n20) {constExponent=5})
   B=Multiply(A=ScalarParameter() {defaultValue=1, parameterName=Brightness}) {constB=6}
  ) {constB=80}
  n22 = Constant() {r=0.5}
  n23 = TextureSample(UVs=SphereMask(A=TextureCoordinate(), B=@n22, Radius=@n22, Hardness=LinearInterpolate(Alpha=@n13.Size) {constA=1, constB=0}) {hardnessPercent=25}) {texture=/Game/Effects/Fort_Effects/Textures/Grad/T_Halo_Corona_Gradient}
  n24 = Desaturation(None=@n23.RGB)
  n25 = Multiply(
   A=Multiply(A=LinearInterpolate(A=@n12, Alpha=@n13.Flicker) {constA=1}, B=@n14.RGB)
   B=Multiply(
    A=MaterialFunctionCall(In=@n21, Contrast=@n18.G) {materialFunction=&CheapContrast}.Result
    B=StaticSwitchParameter(True=@n24, False=@n23.RGB) {parameterName=UseParticleColor}
   )
  )
  n26 = MaterialFunctionCall(LDR Input=Clamp(None=@n25), HDR Tint=@n25) {materialFunction=&DeriveHDRfromLDR}
  MaterialFunctionCall(Emissive=StaticSwitchParameter(True=ShadingPathSwitch(Default=@n26.RemappedRange, Forward=@n25, Mobile=@n25), False=Multiply(A=@n4.Result, B=@n7)) {parameterName=Only_use_the_head_halo}) {materialFunction=&MF_FX_EmissiveAdjustments}
  n27 = MaterialFunctionCall(Distance=ScalarParameter() {defaultValue=200, parameterName=NearCamFade}) {materialFunction=&Near_Camera_Fade}
  n28 = Multiply(A=ScalarParameter() {defaultValue=100, parameterName=DepthFade}, B=@n18.B)
  n29 = Constant() {r=1}
  n30 = Multiply(A=@n14.A, B=ShadingPathSwitch(Default=DepthFade(Opacity=@n27.Result, FadeDistance=@n28), Forward=@n29, Mobile=@n29))
  n31 = Saturate(None=SphereMask(A=TextureCoordinate(), B=Constant2Vector() {g=0.25, r=0.5}) {attenuationRadius=0.55, hardnessPercent=1})
  n32 = NamedRerouteUsage()
  n33 = MaterialFunctionCall(Value_Morning=@n1.A, Value_Day=Constant() {r=1}, Value_Evening=@n2.A, Value_Night=@n3.A) {materialFunction=&Scalar_Time_of_Day}
  Multiply(
   A=StaticSwitchParameter(True=Multiply(A=@n30, B=@n31), False=Multiply(A=@n32, B=@n33.Result)) {parameterName=Only_use_the_head_halo}
   B=ScalarParameter() {defaultValue=1, parameterName=MasterFade}
  )
  n34 = TextureCoordinate() {uTiling=0.5, vTiling=0.1}
  n35 = Panner(Coordinate=Add(A=@n34, B=Constant2Vector() {g=0.05})) {speedX=0.1}
  n36 = Panner(Coordinate=Add(A=@n34, B=Constant2Vector() {g=0.1})) {speedX=-0.1}
  n37 = Power(Base=Multiply(
    A=TextureSample(UVs=@n35) {texture=/Game/Characters/Player/Female/Medium/Bodies/F_MED_Buffet_Shard/FX/T_Flare_Ray}.RGB
    B=TextureSample(UVs=@n36) {texture=&T_Flare_Ray}.RGB
   )) {constExponent=1.5}
  n38 = ComponentMask(None=TextureCoordinate()) {g=true}
  n39 = Power(Base=ComponentMask(None=TextureCoordinate()) {g=true}) {constExponent=1}
  n40 = Multiply(A=Power(Base=OneMinus(None=@n38)) {constExponent=1}, B=Saturate(None=Multiply(A=@n39) {constB=10}))
  n41 = LinearInterpolate(Alpha=TextureSample() {texture=&T_Inky_Smoke_Tile}.R) {constA=0.2, constB=1.5}
  n42 = ComponentMask(None=TextureCoordinate()) {r=true}
  n43 = Multiply(A=@n42, B=OneMinus(None=@n42))
  n44 = Multiply(A=Multiply(A=Multiply(A=@n37) {constB=0.5}, B=Multiply(A=@n40, B=@n41)), B=Saturate(None=Multiply(A=@n43) {constB=2.5}))
  n45 = TextureCoordinate() {uTiling=2, vTiling=0.6}
  n46 = ComponentMask(None=@n45) {r=true}
  n47 = OneMinus(None=ComponentMask(None=@n45) {g=true})
  n48 = Panner(Coordinate=Add(A=AppendVector(A=@n46, B=@n47)) {constB=-0.4}) {speedX=0.3}
  n49 = Constant() {r=1}
  n50 = Multiply(
   A=Multiply(
    A=Multiply(A=@n44) {constB=15}
    B=TextureSample(UVs=@n48) {samplerSource=SSM_Wrap_WorldGroupSettings, texture=/Game/Effects/Fort_Effects/Textures/T_Smoke_Ring_Mask2}.R
   )
   B=ShadingPathSwitch(
    Default=MaterialFunctionCall() {materialFunction=&FadeBasedOnViewAngle}.Result
    Forward=@n49
    Mobile=@n49
   )
  )
  n51 = DepthFade(Opacity=@n50) {fadeDistanceDefault=30}
  NamedRerouteDeclaration(None=ShadingPathSwitch(Default=@n51, Deferred=@n51, Forward=@n50, Mobile=@n50)) {name=Alpha_Mask, nodeColor=<rgba 1 0.4938 0.0916 1>}
  n52 = TextureCoordinate()
  n53 = Constant2Vector() {g=1.6, r=1}
  n54 = Desaturation(
   None=TextureSample(UVs=MaterialFunctionCall(UVs=@n52, Texture Scale=@n53) {materialFunction=&ScaleUVsByCenter}.UVs) {texture=&T_Halo_Corona_Gradient}.RGB
   Fraction=Constant() {r=-2}
  )
  n55 = Constant3Vector() {constant=<rgba 0 0.4567 1 1>}
  n56 = TextureCoordinate() {uTiling=0.3, vTiling=0.2}
  n57 = Constant() {r=0.005}
  n58 = Panner(Coordinate=Add(A=@n56, B=@n57) {constB=0.005}) {speedX=0.05, speedY=-0.1}
  n59 = Panner(Coordinate=@n56) {speedX=0.05, speedY=-0.1}
  n60 = Panner(Coordinate=Subtract(A=@n56, B=@n57)) {speedX=0.05, speedY=-0.1}
  n61 = Power(Base=MaterialFunctionCall(
    X=TextureSample(UVs=@n58) {texture=/Game/Global/Textures/Default/Noise/T_FireNoise_01}.R
    Y=TextureSample(UVs=@n59) {texture=&T_FireNoise_01}.R
    Z=TextureSample(UVs=@n60) {texture=&T_FireNoise_01}.R
   ) {materialFunction=&MakeFloat3}.Result) {constExponent=1.5}
  n62 = Multiply(A=Desaturation(None=Multiply(A=@n61, B=@n61), Fraction=Constant() {r=-0.2})) {constB=3}
  NamedRerouteDeclaration(None=Add(A=Multiply(A=@n54, B=@n55), B=Multiply(A=@n62) {constB=5})) {name=Texture_Panner01, nodeColor=<rgba 0 0.0132 1 1>}
  n63 = TextureCoordinate() {uTiling=0.5, vTiling=0.2}
  n64 = Constant() {r=0.005}
  n65 = Add(A=@n63, B=@n64) {constB=0.005}
  n66 = Constant2Vector() {r=0.5}
  n67 = TextureSample(UVs=Panner(Coordinate=Add(A=@n65, B=@n66)) {speedY=-0.08}) {texture=/Game/Characters/Player/Female/Medium/Bodies/F_MED_Buffet_Shard/FX/T_Wave}
  n68 = TextureSample(UVs=Panner(Coordinate=Add(A=@n63, B=@n66)) {speedY=-0.08}) {texture=&T_Wave}
  n69 = Subtract(A=@n63, B=@n64)
  n70 = TextureSample(UVs=Panner(Coordinate=Add(A=@n69, B=@n66)) {speedY=-0.08}) {texture=&T_Wave}
  n71 = Power(Base=Desaturation(
    None=MaterialFunctionCall(X=@n67.R, Y=@n68.R, Z=@n70.R) {materialFunction=&MakeFloat3}.Result
    Fraction=Constant()
   ))
  NamedRerouteDeclaration(None=Multiply(A=Multiply(A=@n71) {constB=500}, B=Constant3Vector() {constant=<rgba 0.6851 0.4896 1 0>})) {name=Texture_Panner02, nodeColor=<rgba 0 0.0132 1 1>}
  ParticleColor()
 ]
}
