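# Cosmetic Niagara system for the Buffet Shard body, adding purple-blue mesh fragments, smoke, flares and sparkles at shoulder and skirt bones.
# Trailing emitters inherit source movement with falloff and gate spawning on skydiving, in-vehicle and owner-speed checks.
# Attached emitters resample bones every update to stay on the character, with skirt smoke switching to its attached variant while skydiving.
# Time of day scales alpha on all emitters except the sparkles.
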NiagaraSystem /BRCosmetics/Characters/Player/Female/Medium/Bodies/F_MED_Buffet_Shard/FX/NS_Buffet_Shard_Skin_FX
dependencies {
 usedDataInterfaces [&NiagaraDataInterfaceCurve, &NiagaraDataInterfaceVector2DCurve]
 usedDynamicInputs []
 usedModules [
  &self:SystemSpawnScript.SystemScriptSource.SystemScriptGraph.NiagaraNodeAssignment_0.SetVariables_#1
  &SystemState
 ]
}
emitters {
 Drop_Flare {
  simTarget=GPUComputeSim
  emitterUpdateScript [
   &EmitterState()
   &SpawnRate(
    SpawnRate=&MakeCustomFloatFromBool(
     Bool=$User.IsSkydiving
     True Float=0
     False Float=&MakeCustomFloatFromBool(
      Bool=$User.bisInVehicle
      True Float=0
      False Float=&MakeCustomFloatFromBool(
       Bool=&SetBoolByFloatComparison(A=&VectorLength(Vector=$Engine.Owner.Velocity), B=1000)
       True Float=0
       False Float=15
      )
     )
    )
    Use Spawn Probability=false
   )
  ]
  particleSpawnScript [
   &InitializeParticle(
    Color=<rgba 1.5469 2.6507 3 1>
    Lifetime Max=1.5
    Lifetime Mode=::NewEnumerator1
    Mesh Scale Mode=::NewEnumerator1
    Mesh Uniform Scale Max=0.5
    Mesh Uniform Scale Min=0.15
    Sprite Size Mode=::NewEnumerator1
    Uniform Sprite Size Max=12
    Uniform Sprite Size Min=7
    UsePositionOffset=false
    Write Material Random=true
   )
   &SkeletalMeshLocation(Bone Sampling Mode=::NewEnumerator4, MaskSpawnGroup=false, Sampled Position Offset=<-30 0 0>, Skeletal Mesh=$System.Skeletal Mesh)
   &SkeletalMeshLocation(
    Bone / Socket Index=2
    Bone Sampling Mode=::NewEnumerator4
    Interpolate Mesh Sampling=false
    MaskSpawnGroup=false
    Orientation Sampling=::NewEnumerator1
    Position Sampling=::NewEnumerator2
    Skeletal Mesh=$System.Skeletal Mesh
   ) {module=SkeletalMeshLocation001}
   &CylinderLocation(Cylinder Height=15, Cylinder Radius=25, MaskSpawnGroup=false, Override Seed=false, Surface Only=true, UseOwnerScale=true)
   &AddVelocityFromPoint(
    Module.Use Falloff Distance=false
    Origin Offset=<0 0 130>
    Velocity Origin:NiagaraPosition=$Particles.SkeletalMeshLocation001.SampledPosition
    Velocity Strength=&RandomRangeFloat(Minimum=30, Maximum=40)
   )
   &CameraOffset(Camera Offset Amount=5)
  ]
  particleUpdateScript [
   &Drag(Use Linear Drag=true, Use Rotational Drag=false)
   &CurlNoiseForce(
    Noise Strength=&FloatFromCurve(FloatCurve=DataInterface() {propertyValues {Curve {keys [<at 0 100>,<at 1 50>]}}})
    Pan Noise=false
   )
   &ParticleState()
   &GravityForce(Gravity=<0 0 -10>)
   &InheritSourceMovement(
    Applied Position Falloff Curve=DataInterface() {propertyValues {Curve {keys [{interpMode=RCIM_Cubic, tangentMode=RCTM_User, time=0, value=1}, {interpMode=RCIM_Cubic, tangentMode=RCTM_User, time=1, value=0}]}}}
    Applied Position Falloff Distance=150
    ApplyForce=false
    ApplyPosition=true
    ApplyPositionFalloff=true
    ApplyPositionFalloffCurve=true
   )
   &SolveForcesAndVelocity(Clamp Velocity=false, Limit Acceleration=false)
   &ScaleSpriteSize(Uniform Curve Sprite Scale=DataInterface() {propertyValues {Curve {keys [
        {arriveTangent=14.9673, interpMode=RCIM_Cubic, leaveTangent=14.9673, tangentMode=RCTM_User, time=0, value=0}
        <at 0.2 1 RCIM_Cubic>
        {arriveTangent=-3.2414, interpMode=RCIM_Cubic, leaveTangent=-3.2414, tangentMode=RCTM_User, time=1, value=0}
       ]}}}) {module=ScaleSpriteSize002}
   &ScaleColor(
    Scale Alpha=&Multiply_Float(A=&FNTimeOfDayScalar(Night Scale=0.7, Morning Scale=0.1, Day Scale=0.7, Evening Scale=0.5), B=1)
    Scale RGB=&VectorFromFloat(Value=&FloatFromCurve(FloatCurve=DataInterface() {propertyValues {Curve {keys [
          {arriveTangent=85.1377, interpMode=RCIM_Cubic, leaveTangent=85.1377, tangentMode=RCTM_User, time=0, value=0}
          {arriveTangentWeight=1.8705, interpMode=RCIM_Cubic, leaveTangentWeight=1.8705, tangentMode=RCTM_User, time=0.05, value=3}
          <at 0.1 0.5 RCIM_Cubic>
         ]}}}))
    ScaleA=true
    ScaleRGB=true
   ) {module=ScaleColor001}
   &DynamicMaterialParameters(
    Index 0 Param 1=0.5
    Index 0 Param 2=&FloatFromCurve(FloatCurve=DataInterface() {propertyValues {Curve {keys [{arriveTangent=-12.4095, interpMode=RCIM_Cubic, leaveTangent=-12.4095, tangentMode=RCTM_User, time=0.5, value=2}, <at 1 0 RCIM_Cubic>]}}})
    Module.Param0WriteEnabled=true
    Module.Param1WriteEnabled=true
    Module.Param2WriteEnabled=true
    Module.Param3WriteEnabled=true
   ) {module=DynamicMaterialParameters001}
  ]
  rendererClasses [&NiagaraSpriteRendererProperties]
  renderers [{rendererClass=&NiagaraSpriteRendererProperties}]
 }
 Shoulder_Fragments {
  simTarget=GPUComputeSim
  emitterUpdateScript [
   &EmitterState()
   &SpawnRate(
    SpawnRate=&MakeCustomFloatFromBool(
     Bool=$User.IsSkydiving
     True Float=0
     False Float=&MakeCustomFloatFromBool(
      Bool=$User.bisInVehicle
      True Float=0
      False Float=&MakeCustomFloatFromBool(
       Bool=&SetBoolByFloatComparison(A=&VectorLength(Vector=$Engine.Owner.Velocity), B=1000)
       True Float=0
       False Float=7
      )
     )
    )
    Use Spawn Probability=false
   )
  ]
  particleSpawnScript [
   &InitializeParticle(
    Color=&MakeCustomLinearColorFromBool(Bool=&RandomBool(), True Color=<rgba 0.9083 0.1374 1.5 1>, False Color=<rgba 0.1213 0.3179 1.5 1>)
    Lifetime Max=1
    Lifetime Mode=::NewEnumerator1
    Mesh Scale Mode=::NewEnumerator1
    Mesh Uniform Scale Max=0.35
    Mesh Uniform Scale Min=0.15
    Sprite Size Mode=::NewEnumerator3
    Uniform Sprite Size=5
    UsePositionOffset=false
    Write Material Random=true
   )
   &SkeletalMeshLocation(
    Bone / Socket Index=1
    Bone Sampling Mode=::NewEnumerator4
    MaskSpawnGroup=false
    Orientation Sampling=::NewEnumerator1
    Sampled Position Offset=&Add_Vector(A=<20 0 7>, B=<0 0 0>)
    Skeletal Mesh=$System.Skeletal Mesh
   ) {module=SkeletalMeshLocation001}
   &SphereLocation(
    MaskSpawnGroup=false
    Override Seed=false
    Sphere Radius=8
    Surface Only=true
    UseNonUniformScale=false
    UseOffset=true
    UseSphereOrientation=false
   )
   &InitialMeshOrientation(
    Rotation=&RandomRangeVector(Minimum=<0 0 0>, Maximum=<1 1 1>)
    Use Orientation Vector=true
    Use Rotation Vector=true
   ) {module=InitialMeshOrientation001}
   &AddVelocity(
    Scale Added Velocity=&Multiply_Vector(A=<1 1 1>, B=&VectorFromFloat(Value=0.1))
    Velocity=<100 -50 100>
   )
  ]
  particleUpdateScript [
   &Drag(Use Linear Drag=true, Use Rotational Drag=false)
   &CurlNoiseForce(
    Noise Strength=&FloatFromCurve(FloatCurve=DataInterface() {propertyValues {Curve {keys [<at 0 100>,<at 1 50>]}}})
    Pan Noise=false
   )
   &ParticleState()
   &InheritSourceMovement(
    Applied Position Falloff Curve=DataInterface() {propertyValues {Curve {keys [{interpMode=RCIM_Cubic, tangentMode=RCTM_User, time=0, value=1}, {interpMode=RCIM_Cubic, tangentMode=RCTM_User, time=1, value=0}]}}}
    Applied Position Falloff Distance=500
    ApplyForce=false
    ApplyPosition=true
    ApplyPositionFalloff=true
    ApplyPositionFalloffCurve=true
   )
   &UpdateMeshOrientation(
    Rotation Rate=&FloatFromCurve(
     FloatCurve=DataInterface() {propertyValues {Curve {keys [
         {arriveTangent=-2.1159, interpMode=RCIM_Cubic, leaveTangent=-2.1159, tangentMode=RCTM_User, time=0, value=1}
         {interpMode=RCIM_Cubic, tangentMode=RCTM_Break, time=1, value=0.2}
        ]}}}
     Scale Curve=&RandomRangeFloat(Minimum=0.5, Maximum=0.7)
    )
    Rotation Vector=&RandomRangeVector(Minimum=<-1 -1 -1>, Maximum=<1 1 1>)
   )
   &SolveForcesAndVelocity(Clamp Velocity=false, Limit Acceleration=false)
   &ScaleMeshSize(Scale Factor=&VectorFromFloat(Value=&FloatFromCurve(FloatCurve=DataInterface() {propertyValues {Curve {keys [{interpMode=RCIM_Cubic, tangentMode=RCTM_User, time=0, value=1}, {interpMode=RCIM_Cubic, tangentMode=RCTM_User, time=1, value=0}]}}})))
   &ScaleColor(
    Scale Alpha=&Multiply_Float(
     A=&RandomRangeFloat(Minimum=0.2, Maximum=1)
     B=&Multiply_Float(
      A=&FloatFromCurve(FloatCurve=DataInterface() {propertyValues {Curve {keys [
           {interpMode=RCIM_Cubic, tangentMode=RCTM_User, time=0, value=0}
           {arriveTangent=-1, interpMode=RCIM_Cubic, leaveTangent=-1, time=0.2, value=0.95}
           {arriveTangent=-1.7637, interpMode=RCIM_Cubic, leaveTangent=-1.7637, tangentMode=RCTM_User, time=1, value=0}
          ]}}})
      B=&FNTimeOfDayScalar(Night Scale=0.8, Morning Scale=0.3, Evening Scale=0.5)
     )
    )
    Scale RGB=&VectorFromFloat(Value=&RandomRangeFloat(Minimum=5, Maximum=10))
    ScaleA=true
    ScaleRGB=true
   )
  ]
  rendererClasses [&NiagaraMeshRendererProperties]
  renderers [{rendererClass=&NiagaraMeshRendererProperties, rendererIndex=0}]
 }
 Shoulder_Local {
  emitterUpdateScript [
   &EmitterState()
   &SpawnRate(SpawnRate=5, Use Spawn Probability=false)
  ]
  particleSpawnScript [
   &InitializeParticle(
    Color=&MakeCustomLinearColorFromBool(Bool=&RandomBool(), True Color=<rgba 0.9083 0.1374 1.5 1>, False Color=<rgba 0.1213 0.3179 1.5 1>)
    Lifetime Max=1
    Lifetime Mode=::NewEnumerator1
    Mesh Scale Mode=::NewEnumerator1
    Mesh Uniform Scale Max=0.35
    Mesh Uniform Scale Min=0.15
    Sprite Size Mode=::NewEnumerator3
    Uniform Sprite Size=5
    UsePositionOffset=false
    Write Material Random=true
   )
   &SphereLocation(
    MaskSpawnGroup=false
    Override Seed=false
    Sphere Origin:NiagaraPosition=&MakeVector(X=0, Y=0, Z=0)
    Sphere Radius=8
    Surface Only=true
    UseNonUniformScale=false
    UseOffset=true
    UseSphereOrientation=false
   )
   &InitialMeshOrientation(
    Rotation=&RandomRangeVector(Minimum=<0 0 0>, Maximum=<1 1 1>)
    Use Orientation Vector=true
    Use Rotation Vector=true
   ) {module=InitialMeshOrientation001}
   &AddVelocity(
    Scale Added Velocity=&Multiply_Vector(A=<1 1 1>, B=&VectorFromFloat(Value=0.1))
    Velocity=<100 0 100>
   )
  ]
  particleUpdateScript [
   &Drag(Use Linear Drag=true, Use Rotational Drag=false)
   &CurlNoiseForce(
    Noise Strength=&FloatFromCurve(FloatCurve=DataInterface() {propertyValues {Curve {keys [<at 0 100>,<at 1 50>]}}})
    Pan Noise=false
   )
   &ParticleState()
   &UpdateMeshOrientation(
    Rotation Rate=&FloatFromCurve(
     FloatCurve=DataInterface() {propertyValues {Curve {keys [
         {arriveTangent=-2.1159, interpMode=RCIM_Cubic, leaveTangent=-2.1159, tangentMode=RCTM_User, time=0, value=1}
         {interpMode=RCIM_Cubic, tangentMode=RCTM_Break, time=1, value=0.2}
        ]}}}
     Scale Curve=&RandomRangeFloat(Minimum=0.5, Maximum=0.7)
    )
    Rotation Vector=&RandomRangeVector(Minimum=<-1 -1 -1>, Maximum=<1 1 1>)
   )
   &SolveForcesAndVelocity(Clamp Velocity=false, Limit Acceleration=false)
   &SkeletalMeshLocation(
    Bone / Socket Index=1
    Bone Sampling Mode=::NewEnumerator4
    MaskSpawnGroup=false
    Orientation Sampling=::NewEnumerator1
    Sampled Position Offset=&Add_Vector(
     A=<20 0 7>
     B=&Add_Vector(
      A=Linked() {linkedVariable {name=Particles.Initial.Position, type=NiagaraPosition}}
      B=&MakeVector(
       X=&Lerp_Float(A=0, B=10, Alpha=$Particles.NormalizedAge)
       Y=&Lerp_Float(A=0, B=-5, Alpha=$Particles.NormalizedAge)
       Z=&Lerp_Float(A=0, B=10, Alpha=$Particles.NormalizedAge)
      )
     )
    )
    Skeletal Mesh=$System.Skeletal Mesh
   ) {module=SkeletalMeshLocation001}
   &ScaleMeshSize(Scale Factor=&VectorFromFloat(Value=&FloatFromCurve(FloatCurve=DataInterface() {propertyValues {Curve {keys [{interpMode=RCIM_Cubic, tangentMode=RCTM_User, time=0, value=1}, {interpMode=RCIM_Cubic, tangentMode=RCTM_User, time=1, value=0}]}}})))
   &ScaleColor(
    Scale Alpha=&Multiply_Float(
     A=&RandomRangeFloat(Minimum=0.2, Maximum=1)
     B=&Multiply_Float(
      A=&FloatFromCurve(FloatCurve=DataInterface() {propertyValues {Curve {keys [
           {interpMode=RCIM_Cubic, tangentMode=RCTM_User, time=0, value=0}
           {arriveTangent=-1, interpMode=RCIM_Cubic, leaveTangent=-1, time=0.2, value=0.95}
           {arriveTangent=-1.7637, interpMode=RCIM_Cubic, leaveTangent=-1.7637, tangentMode=RCTM_User, time=1, value=0}
          ]}}})
      B=&FNTimeOfDayScalar(Night Scale=0.8, Morning Scale=0.3, Evening Scale=0.5)
     )
    )
    Scale RGB=&VectorFromFloat(Value=&RandomRangeFloat(Minimum=5, Maximum=10))
    ScaleA=true
    ScaleRGB=true
   )
  ]
  rendererClasses [&NiagaraMeshRendererProperties]
  renderers [{rendererClass=&NiagaraMeshRendererProperties, rendererIndex=0}]
 }
 Shoulder_Smoke {
  emitterUpdateScript [
   &EmitterState(
    Enable Distance Culling=true
    MaxDistance=10000
    Module.bUseMaxDistance=true
    Module.bUseMinDistance=false
    Scalability Mode=::NewEnumerator1
   )
   &SpawnRate(SpawnRate=5, Use Spawn Probability=false)
  ]
  particleSpawnScript [
   /SetVariables(Particles.SpriteRotationRate=&RandomRangeFloat(Minimum=-3, Maximum=3, Evaluation Type=::NewEnumerator1)) {isSetParameters=true, module=SetVariables_#2}
   &InitializeParticle(
    Color=&RandomRangeLinearColor(Minimum=<rgba 0.492 0.1577 0.5312 0.4>, Maximum=<rgba 0.0785 0.4503 0.9323 0.2>)
    Lifetime Max=1.5
    Lifetime Min=0.7
    Lifetime Mode=::NewEnumerator1
    Sprite Rotation Mode=::NewEnumerator2
    Sprite Size Mode=::NewEnumerator1
    Uniform Sprite Size Max=50
    Uniform Sprite Size Min=40
    UsePositionOffset=false
    Write Material Random=true
   ) {module=InitializeParticle001}
   &DynamicMaterialParameters(
    Index 0 Param 1=&RandomRangeFloat(Minimum=0, Maximum=1, Evaluation Type=::NewEnumerator1)
    Index 0 Param 2=&RandomRangeFloat(Minimum=0, Maximum=1, Evaluation Type=::NewEnumerator1)
    Module.Param0WriteEnabled=true
    Module.Param1WriteEnabled=true
    Module.Param2WriteEnabled=true
    Module.Param3WriteEnabled=true
   )
   &CameraOffset(Camera Offset Amount=&RandomRangeFloat(Minimum=0, Maximum=10, Evaluation Type=::NewEnumerator1))
   &AddVelocity(
    Scale Added Velocity=&Multiply_Vector(A=<1 1 1>, B=&VectorFromFloat(Value=0.1))
    Velocity=<100 0 50>
   )
   &AddVelocity(
    Coordinate Space=::World
    Scale Added Velocity=&Multiply_Vector(A=<1 1 1>, B=&VectorFromFloat(Value=0.5))
    Velocity=<0 0 6>
   ) {module=AddVelocity001}
  ]
  particleUpdateScript [
   &ParticleState()
   &Drag(Drag=2.5, Use Linear Drag=true, Use Rotational Drag=false)
   &SolveForcesAndVelocity(Clamp Velocity=false, Limit Acceleration=false)
   &SkeletalMeshLocation(
    Bone / Socket Index=1
    Bone Sampling Mode=::NewEnumerator4
    MaskSpawnGroup=false
    Sampled Position Offset=&Add_Vector(
     A=<20 0 2>
     B=&MakeVector(
      X=&Lerp_Float(A=0, B=10, Alpha=$Particles.NormalizedAge)
      Y=&Lerp_Float(A=0, B=-5, Alpha=$Particles.NormalizedAge)
      Z=&Lerp_Float(A=0, B=10, Alpha=$Particles.NormalizedAge)
     )
    )
    Skeletal Mesh=$System.Skeletal Mesh
   )
   &ScaleColor(
    Scale Alpha=&Multiply_Float(
     A=&FloatFromCurve(FloatCurve=DataInterface() {propertyValues {Curve {keys [
          {interpMode=RCIM_Cubic, tangentMode=RCTM_User, time=0, value=0}
          {interpMode=RCIM_Cubic, tangentMode=RCTM_User, time=0.5, value=1}
          {tangentMode=RCTM_User, time=1, value=0}
         ]}}})
     B=&Multiply_Float(A=&FNTimeOfDayScalar(Night Scale=0.8, Morning Scale=0.1, Evening Scale=0.2), B=1)
    )
    ScaleA=true
    ScaleRGB=true
   )
  ]
  rendererClasses [&NiagaraSpriteRendererProperties]
  renderers [{rendererClass=&NiagaraSpriteRendererProperties}]
 }
 Skirt_Fragments {
  emitterUpdateScript [
   &EmitterState()
   &SpawnRate(
    SpawnRate=&MakeCustomFloatFromBool(
     Bool=$User.IsSkydiving
     True Float=0
     False Float=&MakeCustomFloatFromBool(
      Bool=$User.bisInVehicle
      True Float=0
      False Float=&MakeCustomFloatFromBool(
       Bool=&SetBoolByFloatComparison(A=&VectorLength(Vector=$Engine.Owner.Velocity), B=1000)
       True Float=0
       False Float=30
      )
     )
    )
    Use Spawn Probability=false
   )
  ]
  particleSpawnScript [
   &InitializeParticle(
    Color=&MakeCustomLinearColorFromBool(Bool=&RandomBool(), True Color=<rgba 0.9083 0.1374 1.5 1>, False Color=<rgba 0.1213 0.3179 1.5 1>)
    Lifetime Max=1.5
    Lifetime Mode=::NewEnumerator1
    Mesh Scale Mode=::NewEnumerator1
    Mesh Uniform Scale Max=0.5
    Mesh Uniform Scale Min=0.15
    Sprite Size Mode=::NewEnumerator3
    Uniform Sprite Size=5
    UsePositionOffset=false
    Write Material Random=true
   )
   &SkeletalMeshLocation(Bone Sampling Mode=::NewEnumerator4, MaskSpawnGroup=false, Sampled Position Offset=<-35 0 0>, Skeletal Mesh=$System.Skeletal Mesh)
   &SkeletalMeshLocation(
    Bone / Socket Index=2
    Bone Sampling Mode=::NewEnumerator4
    Interpolate Mesh Sampling=false
    MaskSpawnGroup=false
    Orientation Sampling=::NewEnumerator1
    Position Sampling=::NewEnumerator2
    Skeletal Mesh=$System.Skeletal Mesh
   ) {module=SkeletalMeshLocation001}
   &CylinderLocation(Cylinder Height=10, Cylinder Radius=25, MaskSpawnGroup=false, Override Seed=false, Surface Only=true, UseOwnerScale=true)
   &AddVelocityFromPoint(
    Module.Use Falloff Distance=false
    Velocity Origin:NiagaraPosition=$Particles.SkeletalMeshLocation001.SampledPosition
    Velocity Strength=&RandomRangeFloat(Minimum=30, Maximum=40)
   )
   &InitialMeshOrientation(
    Rotation=&RandomRangeVector(Minimum=<0 0 0>, Maximum=<1 1 1>)
    Use Orientation Vector=true
    Use Rotation Vector=true
   ) {module=InitialMeshOrientation001}
   &CameraOffset(Camera Offset Amount=&RandomRangeFloat(Minimum=0, Maximum=20, Evaluation Type=::NewEnumerator1))
  ]
  particleUpdateScript [
   &Drag(Use Linear Drag=true, Use Rotational Drag=false)
   &CurlNoiseForce(
    Noise Strength=&FloatFromCurve(FloatCurve=DataInterface() {propertyValues {Curve {keys [<at 0 100>,<at 1 50>]}}})
    Pan Noise=false
   )
   &ParticleState()
   &GravityForce(Gravity=<0 0 -15>)
   &InheritSourceMovement(
    Applied Position Falloff Curve=DataInterface() {propertyValues {Curve {keys [{interpMode=RCIM_Cubic, tangentMode=RCTM_User, time=0, value=1}, {interpMode=RCIM_Cubic, tangentMode=RCTM_User, time=1, value=0}]}}}
    Applied Position Falloff Distance=200
    ApplyForce=false
    ApplyPosition=true
    ApplyPositionFalloff=true
    ApplyPositionFalloffCurve=true
   )
   &UpdateMeshOrientation(
    Rotation Rate=&FloatFromCurve(
     FloatCurve=DataInterface() {propertyValues {Curve {keys [
         {arriveTangent=-2.1159, interpMode=RCIM_Cubic, leaveTangent=-2.1159, tangentMode=RCTM_User, time=0, value=1}
         {interpMode=RCIM_Cubic, tangentMode=RCTM_Break, time=1, value=0.2}
        ]}}}
     Scale Curve=&RandomRangeFloat(Minimum=0.7, Maximum=1.5)
    )
    Rotation Vector=&RandomRangeVector(Minimum=<-1 -1 -1>, Maximum=<1 1 1>)
   )
   &SolveForcesAndVelocity(Clamp Velocity=false, Limit Acceleration=false)
   &ScaleMeshSize(Scale Factor=&VectorFromFloat(Value=&FloatFromCurve(FloatCurve=DataInterface() {propertyValues {Curve {keys [<at 0.2 1 RCIM_Cubic>, {arriveTangent=-3.8833, interpMode=RCIM_Cubic, tangentMode=RCTM_Break, time=1, value=0}]}}})))
   &ScaleColor(
    Scale Alpha=&Multiply_Float(
     A=&RandomRangeFloat(Minimum=0.2, Maximum=1)
     B=&Multiply_Float(
      A=&FloatFromCurve(FloatCurve=DataInterface() {propertyValues {Curve {keys [
           {interpMode=RCIM_Cubic, tangentMode=RCTM_User, time=0, value=0}
           {arriveTangent=-1, interpMode=RCIM_Cubic, leaveTangent=-1, time=0.2, value=0.95}
           {arriveTangent=-1.7637, interpMode=RCIM_Cubic, leaveTangent=-1.7637, tangentMode=RCTM_User, time=1, value=0}
          ]}}})
      B=&FNTimeOfDayScalar(Night Scale=0.8, Morning Scale=0.3, Evening Scale=0.5)
     )
    )
    Scale RGB=&VectorFromFloat(Value=&RandomRangeFloat(Minimum=5, Maximum=10))
    ScaleA=true
    ScaleRGB=true
   )
  ]
  rendererClasses [&NiagaraMeshRendererProperties]
  renderers [{rendererClass=&NiagaraMeshRendererProperties, rendererIndex=0}]
 }
 Skirt_Smoke {
  emitterUpdateScript [
   &EmitterState(
    Enable Distance Culling=true
    MaxDistance=10000
    Module.bUseMaxDistance=true
    Module.bUseMinDistance=false
    Scalability Mode=::NewEnumerator1
   )
   &SpawnRate(
    SpawnRate=&MakeCustomFloatFromBool(
     Bool=$User.IsSkydiving
     True Float=0
     False Float=&MakeCustomFloatFromBool(
      Bool=$User.bisInVehicle
      True Float=0
      False Float=&MakeCustomFloatFromBool(
       Bool=&SetBoolByFloatComparison(A=&VectorLength(Vector=$Engine.Owner.Velocity), B=1000)
       True Float=0
       False Float=10
      )
     )
    )
    Use Spawn Probability=false
   )
  ]
  particleSpawnScript [
   /SetVariables(Particles.SpriteRotationRate=&RandomRangeFloat(Minimum=-3, Maximum=3, Evaluation Type=::NewEnumerator1)) {isSetParameters=true, module=SetVariables_#2}
   &InitializeParticle(
    Color=&RandomRangeLinearColor(Minimum=<rgba 0.421 0.2362 0.4427 1>, Maximum=<rgba 0 0.2228 0.6042 1>)
    Lifetime Max=1
    Lifetime Min=0.7
    Lifetime Mode=::NewEnumerator1
    Sprite Rotation Mode=::NewEnumerator2
    Sprite Size Mode=::NewEnumerator1
    Uniform Sprite Size Max=40
    Uniform Sprite Size Min=30
    UsePositionOffset=false
    Write Material Random=true
   ) {module=InitializeParticle001}
   &SkeletalMeshLocation(
    Bone Sampling Mode=::NewEnumerator4
    MaskSpawnGroup=false
    Orientation Sampling=::NewEnumerator1
    Sampled Position Offset=<-20 0 0>
    Skeletal Mesh=$System.Skeletal Mesh
   )
   &SkeletalMeshLocation(
    Bone / Socket Index=2
    Bone Sampling Mode=::NewEnumerator4
    Interpolate Mesh Sampling=false
    MaskSpawnGroup=false
    Orientation Sampling=::NewEnumerator1
    Position Sampling=::NewEnumerator2
    Skeletal Mesh=$System.Skeletal Mesh
   ) {module=SkeletalMeshLocation001}
   &CylinderLocation(Cylinder Height=10, Cylinder Radius=20, MaskSpawnGroup=false, Override Seed=false, Surface Only=true, UseOwnerScale=true)
   &AddVelocityFromPoint(
    Module.Use Falloff Distance=false
    Velocity Origin:NiagaraPosition=$Particles.SkeletalMeshLocation001.SampledPosition
    Velocity Strength=&RandomRangeFloat(Minimum=20, Maximum=30)
   )
   &DynamicMaterialParameters(
    Index 0 Param 1=&RandomRangeFloat(Minimum=0, Maximum=1, Evaluation Type=::NewEnumerator1)
    Index 0 Param 2=&RandomRangeFloat(Minimum=0, Maximum=1, Evaluation Type=::NewEnumerator1)
    Module.Param0WriteEnabled=true
    Module.Param1WriteEnabled=true
    Module.Param2WriteEnabled=true
    Module.Param3WriteEnabled=true
   )
   &CameraOffset(Camera Offset Amount=&RandomRangeFloat(Minimum=0, Maximum=50, Evaluation Type=::NewEnumerator1))
   &AddVelocity(Velocity=<0 0 -20>)
  ]
  particleUpdateScript [
   &GravityForce(Gravity=<0 0 -20>)
   &ParticleState()
   &Drag(Drag=2.5, Use Linear Drag=true, Use Rotational Drag=false)
   &InheritSourceMovement(
    Applied Position Falloff Curve=DataInterface() {propertyValues {Curve {keys [{interpMode=RCIM_Cubic, tangentMode=RCTM_User, time=0, value=1}, {interpMode=RCIM_Cubic, tangentMode=RCTM_User, time=1, value=0}]}}}
    Applied Position Falloff Distance=400
    ApplyForce=false
    ApplyPosition=true
    ApplyPositionFalloff=true
    ApplyPositionFalloffCurve=true
   )
   &SolveForcesAndVelocity(Clamp Velocity=false, Limit Acceleration=false)
   &SpriteRotationRate(Delta Time:NiagaraFloat=$Particles.NormalizedAge, Rotation Rate=$Particles.SpriteRotationRate)
   &ScaleColor(
    Scale Alpha=&Multiply_Float(
     A=&FloatFromCurve(FloatCurve=DataInterface() {propertyValues {Curve {keys [
          {interpMode=RCIM_Cubic, tangentMode=RCTM_User, time=0, value=0}
          {interpMode=RCIM_Cubic, tangentMode=RCTM_User, time=0.5, value=1}
          {tangentMode=RCTM_User, time=1, value=0}
         ]}}})
     B=&Multiply_Float(A=&FNTimeOfDayScalar(Night Scale=0.6, Morning Scale=0.15, Evening Scale=0.3), B=1)
    )
    ScaleA=true
    ScaleRGB=true
   )
   &ScaleSpriteSize(Uniform Curve Sprite Scale=DataInterface() {propertyValues {Curve {keys [<at 0 0.8>, {arriveTangent=-1.1268, leaveTangent=-1.4471, time=1, value=1}]}}})
  ]
  rendererClasses [&NiagaraSpriteRendererProperties]
  renderers [{rendererClass=&NiagaraSpriteRendererProperties}]
 }
 Skirt_Smoke_Local {
  emitterUpdateScript [
   &EmitterState()
   &SpawnRate(
    SpawnRate=&MakeCustomFloatFromBool(Bool=$User.IsSkydiving, True Float=15, False Float=0)
    Use Spawn Probability=false
   )
  ]
  particleSpawnScript [
   /SetVariables(Particles.SpriteRotationRate=&RandomRangeFloat(Minimum=-3, Maximum=3, Evaluation Type=::NewEnumerator1)) {isSetParameters=true, module=SetVariables_#2}
   &InitializeParticle(
    Color=&RandomRangeLinearColor(Minimum=<rgba 0.421 0.2362 0.4427 1>, Maximum=<rgba 0 0.2228 0.6042 1>)
    Lifetime Max=1
    Lifetime Min=0.7
    Lifetime Mode=::NewEnumerator1
    Sprite Rotation Mode=::NewEnumerator2
    Sprite Size Mode=::NewEnumerator1
    Uniform Sprite Size Max=40
    Uniform Sprite Size Min=30
    UsePositionOffset=false
    Write Material Random=true
   ) {module=InitializeParticle001}
   &CylinderLocation(
    Cylinder Height=20
    Cylinder Origin:NiagaraPosition=&MakeVector(X=0, Y=0, Z=0)
    Cylinder Radius=20
    MaskSpawnGroup=false
    Override Seed=false
    Surface Only=true
    UseOwnerScale=true
   )
   &DynamicMaterialParameters(
    Index 0 Param 1=&RandomRangeFloat(Minimum=0, Maximum=1, Evaluation Type=::NewEnumerator1)
    Index 0 Param 2=&RandomRangeFloat(Minimum=0, Maximum=1, Evaluation Type=::NewEnumerator1)
    Module.Param0WriteEnabled=true
    Module.Param1WriteEnabled=true
    Module.Param2WriteEnabled=true
    Module.Param3WriteEnabled=true
   )
   &CameraOffset(Camera Offset Amount=&RandomRangeFloat(Minimum=0, Maximum=50, Evaluation Type=::NewEnumerator1))
  ]
  particleUpdateScript [
   &ParticleState()
   &SolveForcesAndVelocity(Clamp Velocity=false, Limit Acceleration=false)
   &SkeletalMeshLocation(
    Bone Sampling Mode=::NewEnumerator4
    MaskSpawnGroup=false
    Orientation Sampling=::NewEnumerator1
    Sampled Position Offset=&Add_Vector(
     A=Linked() {linkedVariable {name=Particles.Initial.Position, type=NiagaraPosition}}
     B=&Add_Vector(
      A=<-20 0 0>
      B=&MakeVector(X=&Lerp_Float(A=0, B=-5, Alpha=$Particles.NormalizedAge), Y=0, Z=0)
     )
    )
    Skeletal Mesh=$User.SK_Mesh
   )
   &SpriteRotationRate(Delta Time:NiagaraFloat=$Particles.NormalizedAge, Rotation Rate=$Particles.SpriteRotationRate)
   &ScaleColor(
    Scale Alpha=&Multiply_Float(
     A=&FloatFromCurve(FloatCurve=DataInterface() {propertyValues {Curve {keys [
          {interpMode=RCIM_Cubic, tangentMode=RCTM_User, time=0, value=0}
          {interpMode=RCIM_Cubic, tangentMode=RCTM_User, time=0.5, value=1}
          {tangentMode=RCTM_User, time=1, value=0}
         ]}}})
     B=&Multiply_Float(A=&FNTimeOfDayScalar(Night Scale=0.8, Morning Scale=0.15, Evening Scale=0.3), B=1)
    )
    ScaleA=true
    ScaleRGB=true
   )
   &ScaleSpriteSize(Uniform Curve Sprite Scale=DataInterface() {propertyValues {Curve {keys [<at 0 0.8>, {arriveTangent=-1.1268, leaveTangent=-1.4471, time=1, value=1}]}}})
  ]
  rendererClasses [&NiagaraSpriteRendererProperties]
  renderers [{rendererClass=&NiagaraSpriteRendererProperties}]
 }
 Skirt_Static_Sparkles {
  simTarget=GPUComputeSim
  emitterUpdateScript [
   &EmitterState()
   &SpawnRate(SpawnRate=10, Use Spawn Probability=false)
  ]
  particleSpawnScript [
   &FN_SetSpawnGroup(Number of Spawn Groups=8)
   &InitializeParticle(
    Color=<rgba 0.7578 1.3216 1.5 1>
    Lifetime Max=1
    Lifetime Min=0.25
    Lifetime Mode=::NewEnumerator1
    Sprite Rotation Mode=::NewEnumerator2
    Sprite Size Mode=::NewEnumerator1
    UsePositionOffset=false
    Write Material Random=true
   )
   &CameraOffset(Camera Offset Amount=10)
   &SphereLocation(
    MaskSpawnGroup=false
    Override Seed=false
    Sphere Origin:NiagaraPosition=&MakeVector(X=0, Y=0, Z=0)
    Sphere Radius=15
    Surface Only=true
    UseNonUniformScale=false
    UseOffset=false
    UseSphereOrientation=false
   )
  ]
  particleUpdateScript [
   &ParticleState()
   &SkeletalMeshLocation(
    Bone Sampling Mode=::NewEnumerator4
    MaskSpawnGroup=false
    Sampled Position Offset:NiagaraPosition=$Particles.Initial.Position
    Skeletal Mesh=$System.Skeletal Mesh
   )
   &DynamicMaterialParameters(
    Index 0 Param 1=0.5
    Index 0 Param 2=&FloatFromCurve(FloatCurve=DataInterface() {propertyValues {Curve {keys [{arriveTangent=-12.4095, interpMode=RCIM_Cubic, leaveTangent=-12.4095, tangentMode=RCTM_User, time=0.5, value=2}, <at 1 0 RCIM_Cubic>]}}})
    Module.Param0WriteEnabled=true
    Module.Param1WriteEnabled=true
    Module.Param2WriteEnabled=true
    Module.Param3WriteEnabled=true
   )
   &ScaleSpriteSize(Uniform Curve Sprite Scale=DataInterface() {propertyValues {Curve {keys [
        {arriveTangent=14.9673, interpMode=RCIM_Cubic, leaveTangent=14.9673, tangentMode=RCTM_User, time=0, value=0}
        <at 0.2 1 RCIM_Cubic>
        {arriveTangent=-3.2414, interpMode=RCIM_Cubic, leaveTangent=-3.2414, tangentMode=RCTM_User, time=1, value=0}
       ]}}})
   &ScaleColor(
    Scale Alpha=&RandomRangeFloat(Minimum=0.2, Maximum=1)
    Scale RGB=&VectorFromFloat(Value=&FloatFromCurve(FloatCurve=DataInterface() {propertyValues {Curve {keys [
          {arriveTangent=321.9435, interpMode=RCIM_Cubic, leaveTangent=321.9435, tangentMode=RCTM_User, time=0, value=0}
          {arriveTangentWeight=1.8705, interpMode=RCIM_Cubic, leaveTangentWeight=1.8705, tangentMode=RCTM_User, time=0.05, value=3}
          <at 0.1 1 RCIM_Cubic>
         ]}}}))
    ScaleA=true
    ScaleRGB=true
   )
  ]
  rendererClasses [&NiagaraSpriteRendererProperties]
  renderers [{rendererClass=&NiagaraSpriteRendererProperties}]
 }
}
summary {
 systemName=NS_Buffet_Shard_Skin_FX
 emitters [
  {emitterName=Skirt_Fragments, rendererClasses [&NiagaraMeshRendererProperties]}
  {emitterName=Skirt_Smoke, rendererClasses [&NiagaraSpriteRendererProperties]}
  {emitterName=Drop_Flare, simTarget=GPUComputeSim, rendererClasses [&NiagaraSpriteRendererProperties]}
  {emitterName=Skirt_Static_Sparkles, simTarget=GPUComputeSim, rendererClasses [&NiagaraSpriteRendererProperties]}
  {emitterName=Shoulder_Smoke, rendererClasses [&NiagaraSpriteRendererProperties]}
  {emitterName=Shoulder_Fragments, simTarget=GPUComputeSim, rendererClasses [&NiagaraMeshRendererProperties]}
  {emitterName=Shoulder_Local, rendererClasses [&NiagaraMeshRendererProperties]}
  {emitterName=Skirt_Smoke_Local, rendererClasses [&NiagaraSpriteRendererProperties]}
 ]
 userVariables [
  User.bisInVehicle {type=NiagaraBool}
  User.IsSkydiving {type=NiagaraBool}
  User.SK_Mesh {
   type=NiagaraDataInterfaceSkeletalMesh
   defaultValue {
    struct=NiagaraExt_VariableValue_DataInterface
    value {dataInterface=&self:NiagaraDataInterfaceSkeletalMesh_1}
   }
  }
 ]
}
