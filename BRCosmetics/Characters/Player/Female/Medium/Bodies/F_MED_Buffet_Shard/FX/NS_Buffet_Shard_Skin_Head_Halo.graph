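# Niagara cosmetic effect for the Buffet Shard skin's head, drawing a two-layer corona halo and a pair of glowing eyes.
# One emitter bursts two spawn groups of two sprites; group masks and partition arrays set color, size, material parameters and offsets.
# Sprites follow the skeletal mesh every update, with alpha scaled by time of day.
# Eyes pulse in alpha and size over the system loop and brighten when the front-end user flag is set.
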NiagaraSystem /BRCosmetics/Characters/Player/Female/Medium/Bodies/F_MED_Buffet_Shard/FX/NS_Buffet_Shard_Skin_Head_Halo
dependencies {
 usedDataInterfaces [
  &NiagaraDataInterfaceArrayColor
  &NiagaraDataInterfaceArrayFloat
  &NiagaraDataInterfaceArrayFloat3
  &NiagaraDataInterfaceCurve
  &NiagaraDataInterfaceSkeletalMesh
 ]
 usedDynamicInputs []
 usedModules [
  &self:SystemSpawnScript.SystemScriptSource.SystemScriptGraph.NiagaraNodeAssignment_0.SetVariables_#1
  &SystemState
 ]
}
emitters {Corona_Halo_and_Eyes {
  emitterUpdateScript [
   &EmitterState(
    Inactive Response=::NewEnumerator1
    Life Cycle Mode=::NewEnumerator1
    Loop Behavior=::NewEnumerator1
    Loop Duration Mode=::NewEnumerator1
    UseLoopDelay=false
   )
   &SpawnBurst_Instantaneous(Spawn Count=2, Use Spawn Probability=false, UseLoopCountLimit=false)
   &SpawnBurst_Instantaneous(Spawn Count=2, Spawn Group=1, Use Spawn Probability=false, UseLoopCountLimit=false) {module=SpawnBurst_Instantaneous001}
  ]
  particleSpawnScript [
   /SetVariables(Particles.VisibilityTag=&SpawnGroup()) {isSetParameters=true}
   &PartitionParticles()
   &InitializeParticle(
    Color=&MaskLinearColorBySpawnGroup(
     Masked Linear Color=&SelectLinearColorFromArray(
      Array Sampling Mode=::NewEnumerator1
      Direct Array Index=$Particles.PartitionParticles.Partition
      Color Selection Array=DataInterface() {propertyValues {ColorData [<rgba 0 1.231 2 0.3>, <rgba 2 0.9846 0.8449 1>]}}
      Random Seed=?
     )
     Passthrough Linear Color=<rgba 1.9341 1.9798 2 1>
    )
    Sprite Size=&MaskVector2dBySpawnGroup(
     Masked Vector2D=&Vector2DFromFloat(Value=&SelectFloatFromArray(Direct Array Index=$Particles.PartitionParticles.Partition, Float Selection Array=DataInterface() {propertyValues {FloatData [44,41.36]}}))
     Passthrough Vector2D=<12 8>
     Spawn Group Mask=0
    )
    Sprite Size Mode=::NewEnumerator4
    UsePositionOffset=false
    Write Material Random=true
   )
   &DynamicMaterialParameters(
    Index 0 Param 1=&MaskFloatBySpawnGroup(Masked Float=0.7, Passthrough Float=1)
    Index 0 Param 2=&MaskFloatBySpawnGroup(Masked Float=0.2, Passthrough Float=1)
    Index 0 Param 3=&MaskFloatBySpawnGroup(Masked Float=0.5, Passthrough Float=1)
    Index 0 Param 4=&MaskFloatBySpawnGroup(Masked Float=1, Passthrough Float=1)
    Index 1 Param 1=&MaskFloatBySpawnGroup(Masked Float=1, Passthrough Float=5)
    Module.Index0Param0Write=true
    Module.Index0Param1Write=true
    Module.Index0Param2Write=true
    Module.Index0Param3Write=true
    Module.Index1Param0Write=true
    Module.Index1Param1Write=false
    Module.Index1Param2Write=false
    Module.Index1Param3Write=false
    Write Parameter Index 1=true
   )
   &CameraOffset(Camera Offset Amount=&MaskFloatBySpawnGroup(Masked Float=0, Passthrough Float=1.5))
  ]
  particleUpdateScript [
   &ParticleState(Kill Particles When Lifetime Has Elapsed=false, Let Infinitely Lived Particles Die When Emitter Deactivates=true)
   &SkeletalMeshLocation(
    Bone Sampling Mode=::NewEnumerator4
    MaskSpawnGroup=false
    Sampled Position Offset=&MaskVectorBySpawnGroup(
     Masked Vector=<5 0 0>
     Passthrough Vector=&SelectVectorFromArray(
      Direct Array Index=$Particles.PartitionParticles.Partition
      Vector Selection Array=DataInterface() {propertyValues {FloatData [<2.5 6.2 3>, <2.5 6.2 -3>]}}
     )
    )
    Skeletal Mesh=$System.Skeletal Mesh
   )
   &ScaleColor(
    Scale Alpha=&MaskFloatBySpawnGroup(
     Masked Float=&MakeCustomFloatFromBool(
      Bool=&SetBoolByIntComparison(A=$Particles.PartitionParticles.Partition, B=0, Comparison Type=::NewEnumerator2)
      True Float=&FNTimeOfDayScalar(Night Scale=1, Morning Scale=0.2, Evening Scale=0.5)
      False Float=&Multiply_Float(A=0.3, B=&FNTimeOfDayScalar(Night Scale=1, Morning Scale=0.2, Evening Scale=0.5))
     )
     Passthrough Float=&Multiply_Float(
      A=&Multiply_Float(
       A=&MakeCustomFloatFromBool(Bool=$User.Is_it_in_FrontEnd, True Float=2.7, False Float=1)
       B=&FNTimeOfDayScalar(Night Scale=0.8, Morning Scale=0.3, Evening Scale=0.5)
      )
      B=&FloatFromCurve(
       FloatCurve=DataInterface() {propertyValues {Curve {keys [<at 0 1>,<at 0.0539 1>,<at 0.1557 0.8>,<at 0.2614 1>,<at 0.3742 0.8>,<at 0.456 1>,<at 0.5872 0.8>,<at 0.7535 1>,<at 0.8976 0.8>,<at 1 1>]}}}
       CurveIndex=$System.NormalizedLoopAge
      )
     )
    )
    Scale RGB=&MaskVectorBySpawnGroup(Masked Vector=<1 1 1>, Passthrough Vector=<1.3 1.3 1.3>)
    ScaleA=true
    ScaleRGB=true
   )
   &ScaleSpriteSize(
    Scale Sprite Size Mode=::NewEnumerator0
    Uniform Scale Factor=&MaskFloatBySpawnGroup(
     Masked Float=1
     Passthrough Float=&FloatFromCurve(
      FloatCurve=DataInterface() {propertyValues {Curve {keys [<at 0 1>,<at 0.0539 1>,<at 0.1557 0.985>,<at 0.2614 1>,<at 0.3742 0.985>,<at 0.456 1>,<at 0.5872 0.985>,<at 0.7535 1>,<at 0.8976 0.985>,<at 1 1>]}}}
      CurveIndex=$System.NormalizedLoopAge
     )
    )
   )
  ]
  rendererClasses [&NiagaraSpriteRendererProperties]
  renderers [
   {rendererClass=&NiagaraSpriteRendererProperties}
   {rendererClass=&NiagaraSpriteRendererProperties, rendererIndex=1}
  ]
 }}
summary {
 systemName=NS_Buffet_Shard_Skin_Head_Halo
 emitters [{emitterName=Corona_Halo_and_Eyes, rendererClasses [&NiagaraSpriteRendererProperties]}]
 userVariables [User.Is_it_in_FrontEnd {type=NiagaraBool}]
}
