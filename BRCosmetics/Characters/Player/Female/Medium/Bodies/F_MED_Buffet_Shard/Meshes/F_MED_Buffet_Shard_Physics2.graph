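# Physics asset for the F_MED_Buffet_Shard female body, defining capsule bodies and joints for the skirt's dynamic bones, the pelvis and both thighs.
# Front, side and back skirt base bodies and both thighs joint to the pelvis with free swing and twist.
# Skirt chains descend from the left and right bases with tight angular limits.
# Only right-side chains have a third segment, with twist locked on front and side.
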PhysicsAsset /BRCosmetics/Characters/Player/Female/Medium/Bodies/F_MED_Buffet_Shard/Meshes/F_MED_Buffet_Shard_Physics2
bodies {
 dyn_skirt_bk_a_base [{shapeType=Capsule, length=1, radius=1}]
 dyn_skirt_bk_a_base_l [{shapeType=Capsule, length=1, radius=1}]
 dyn_skirt_bk_a_base_r [{shapeType=Capsule, length=1, radius=1}]
 dyn_skirt_bk_a_l [{shapeType=Capsule, length=2.5, radius=5, center=<7 0 4>, rotation=<rot 90 0 0>}]
 dyn_skirt_bk_a_r [{shapeType=Capsule, length=2.5, radius=5, center=<-7 0 -4>, rotation=<rot 90 0 180>}]
 dyn_skirt_bk_b_l [{shapeType=Capsule, length=6, radius=7, center=<9 0 4>, rotation=<rot 90 0 0>}]
 dyn_skirt_bk_b_r [{shapeType=Capsule, length=6, radius=7, center=<-9 0 -4>, rotation=<rot 90 0 180>}]
 dyn_skirt_bk_c_r [{shapeType=Capsule, length=7.25, radius=3.75, center=<-6.9305 0 0>, rotation=<rot 90 0 0>}]
 dyn_skirt_fr_a_base [{shapeType=Capsule, length=1, radius=1}]
 dyn_skirt_fr_a_base_l [{shapeType=Capsule, length=1, radius=1}]
 dyn_skirt_fr_a_base_r [{shapeType=Capsule, length=1, radius=1}]
 dyn_skirt_fr_a_l [{shapeType=Capsule, length=2.5, radius=5, center=<7 0 4>, rotation=<rot 90 0 0>}]
 dyn_skirt_fr_a_r [{shapeType=Capsule, length=2.5, radius=5, center=<-7 0 -4>, rotation=<rot 90 0 180>}]
 dyn_skirt_fr_b_l [{shapeType=Capsule, length=6, radius=6, center=<9 0 4>, rotation=<rot 90 0 0>}]
 dyn_skirt_fr_b_r [{shapeType=Capsule, length=6, radius=6, center=<-9 0 -4>, rotation=<rot 90 0 180>}]
 dyn_skirt_fr_c_r [{shapeType=Capsule, length=9.25, radius=9, center=<-11.257 0 -6>, rotation=<rot 90 0 0>}]
 dyn_skirt_side_a_base_l [{shapeType=Capsule, length=1, radius=1}]
 dyn_skirt_side_a_base_r [{shapeType=Capsule, length=1, radius=1}]
 dyn_skirt_side_a_l [{shapeType=Capsule, length=2.5, radius=5, center=<7 0 4>, rotation=<rot 90 0 0>}]
 dyn_skirt_side_a_r [{shapeType=Capsule, length=2.5, radius=5, center=<-7 0 -4>, rotation=<rot 90 0 180>}]
 dyn_skirt_side_b_l [{shapeType=Capsule, length=6, radius=7.5, center=<9 0 4>, rotation=<rot 90 0 0>}]
 dyn_skirt_side_b_r [{shapeType=Capsule, length=6, radius=7.5, center=<-9 0 -4>, rotation=<rot 90 0 180>}]
 dyn_skirt_side_c_r [{shapeType=Capsule, length=9.25, radius=9, center=<-11.257 0 -6>, rotation=<rot 90 0 0>}]
 pelvis [{shapeType=Capsule, length=8.5, radius=7.25, center=<-1.7235 0 0>}]
 thigh_l [{shapeType=Capsule, length=27.75, radius=7.25, center=<-16.9134 -0.0001 0.6245>, rotation=<rot 90 0 0>}]
 thigh_r [{shapeType=Capsule, length=27.5, radius=7.5, center=<16.9134 0.0001 -0.2858>, rotation=<rot 90 0 180>}]
}
constraints [
 {
  bone1Name=dyn_skirt_side_a_base_l
  bone2Name=pelvis
  swing1LimitDegrees=45
  swing1Motion=Free
  swing2LimitDegrees=45
  swing2Motion=Free
  twistLimitDegrees=45
  twistMotion=Free
 }
 {
  bone1Name=dyn_skirt_bk_a_base_l
  bone2Name=pelvis
  swing1LimitDegrees=45
  swing1Motion=Free
  swing2LimitDegrees=45
  swing2Motion=Free
  twistLimitDegrees=45
  twistMotion=Free
 }
 {
  bone1Name=dyn_skirt_fr_a_base_l
  bone2Name=pelvis
  swing1LimitDegrees=45
  swing1Motion=Free
  swing2LimitDegrees=45
  swing2Motion=Free
  twistLimitDegrees=45
  twistMotion=Free
 }
 {
  bone1Name=dyn_skirt_side_a_base_r
  bone2Name=pelvis
  swing1LimitDegrees=45
  swing1Motion=Free
  swing2LimitDegrees=45
  swing2Motion=Free
  twistLimitDegrees=45
  twistMotion=Free
 }
 {
  bone1Name=dyn_skirt_bk_a_base_r
  bone2Name=pelvis
  swing1LimitDegrees=45
  swing1Motion=Free
  swing2LimitDegrees=45
  swing2Motion=Free
  twistLimitDegrees=45
  twistMotion=Free
 }
 {
  bone1Name=dyn_skirt_fr_a_base_r
  bone2Name=pelvis
  swing1LimitDegrees=45
  swing1Motion=Free
  swing2LimitDegrees=45
  swing2Motion=Free
  twistLimitDegrees=45
  twistMotion=Free
 }
 {
  bone1Name=dyn_skirt_fr_a_base
  bone2Name=pelvis
  swing1LimitDegrees=45
  swing1Motion=Free
  swing2LimitDegrees=45
  swing2Motion=Free
  twistLimitDegrees=45
  twistMotion=Free
 }
 {
  bone1Name=dyn_skirt_bk_a_base
  bone2Name=pelvis
  swing1LimitDegrees=45
  swing1Motion=Free
  swing2LimitDegrees=45
  swing2Motion=Free
  twistLimitDegrees=45
  twistMotion=Free
 }
 {
  bone1Name=thigh_l
  bone2Name=pelvis
  swing1LimitDegrees=45
  swing1Motion=Free
  swing2LimitDegrees=45
  swing2Motion=Free
  twistLimitDegrees=45
  twistMotion=Free
 }
 {
  bone1Name=thigh_r
  bone2Name=pelvis
  swing1LimitDegrees=45
  swing1Motion=Free
  swing2LimitDegrees=45
  swing2Motion=Free
  twistLimitDegrees=45
  twistMotion=Free
 }
 {bone1Name=dyn_skirt_side_a_l, bone2Name=dyn_skirt_side_a_base_l, swing1LimitDegrees=5, swing2LimitDegrees=20, twistLimitDegrees=5}
 {bone1Name=dyn_skirt_side_b_l, bone2Name=dyn_skirt_side_a_l, swing1LimitDegrees=5, swing2LimitDegrees=20, twistLimitDegrees=5}
 {bone1Name=dyn_skirt_bk_b_l, bone2Name=dyn_skirt_bk_a_l, swing1LimitDegrees=10, swing2LimitDegrees=20, twistLimitDegrees=5}
 {bone1Name=dyn_skirt_bk_a_l, bone2Name=dyn_skirt_bk_a_base_l, swing1LimitDegrees=5, swing2LimitDegrees=20, twistLimitDegrees=5}
 {bone1Name=dyn_skirt_fr_a_l, bone2Name=dyn_skirt_fr_a_base_l, swing1LimitDegrees=5, swing2LimitDegrees=15, twistLimitDegrees=5}
 {bone1Name=dyn_skirt_fr_b_l, bone2Name=dyn_skirt_fr_a_l, swing1LimitDegrees=5, swing2LimitDegrees=15, twistLimitDegrees=5}
 {bone1Name=dyn_skirt_fr_a_r, bone2Name=dyn_skirt_fr_a_base_r, swing1LimitDegrees=5, swing2LimitDegrees=15, twistLimitDegrees=5}
 {bone1Name=dyn_skirt_side_a_r, bone2Name=dyn_skirt_side_a_base_r, swing1LimitDegrees=5, swing2LimitDegrees=20, twistLimitDegrees=5}
 {bone1Name=dyn_skirt_bk_a_r, bone2Name=dyn_skirt_bk_a_base_r, swing1LimitDegrees=5, swing2LimitDegrees=20, twistLimitDegrees=5}
 {bone1Name=dyn_skirt_bk_b_r, bone2Name=dyn_skirt_bk_a_r, swing1LimitDegrees=10, swing2LimitDegrees=20, twistLimitDegrees=5}
 {bone1Name=dyn_skirt_fr_b_r, bone2Name=dyn_skirt_fr_a_r, swing1LimitDegrees=5, swing2LimitDegrees=15, twistLimitDegrees=5}
 {bone1Name=dyn_skirt_side_b_r, bone2Name=dyn_skirt_side_a_r, swing1LimitDegrees=5, swing2LimitDegrees=20, twistLimitDegrees=5}
 {
  bone1Name=dyn_skirt_fr_c_r
  bone2Name=dyn_skirt_fr_b_r
  swing1LimitDegrees=20
  swing2LimitDegrees=45
  twistLimitDegrees=5
  twistMotion=Locked
 }
 {bone1Name=dyn_skirt_bk_c_r, bone2Name=dyn_skirt_bk_b_r, swing1LimitDegrees=5, swing2LimitDegrees=25, twistLimitDegrees=5}
 {
  bone1Name=dyn_skirt_side_c_r
  bone2Name=dyn_skirt_side_b_r
  swing1LimitDegrees=20
  swing2LimitDegrees=45
  twistLimitDegrees=5
  twistMotion=Locked
 }
]
