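# Post-process style AnimBP for the F_MED_Buffet_Shard female body cosmetic, adding secondary motion to its shards and FX bones.
# Copies the parent mesh pose through a muscle-volume Control Rig driven by pose-deformer inputs, caching stages between passes.
# Weighted constraints anchor shoulder and skirt shards to body and skirt bones before AnimDynamics springs them.
# State bools offset shoulder shards when skydiving, shrink some FX bones in vehicles, and pin simspace to pelvis in front end.
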AnimBlueprint /BRCosmetics/Characters/Player/Female/Medium/Bodies/F_MED_Buffet_Shard/Meshes/F_MED_Buffet_Shard_AnimBP
target_skeleton=&/F_MED_Buffet_Shard_Skeleton
graphs [
 AnimGraph {nodes [
   n1 = Anim Dynamics - Bone: dyn_fx_wide_a_r(ComponentPose=Local To Component(LocalPose=Use cached pose 'CrystalCachedPose'()), Alpha=1, ExternalForce="0,0,0") {
    angularDampingOverride=1
    angularSpringConstant=20000
    bAngularSpring=true
    bGravityOverrideInSimSpace=true
    bOverrideAngularDamping=true
    bUseGravityOverride=true
    gravityScale=0.35
    simulationSpace=RootRelative
    alphaBoolBlend {blendOption=Linear}
    boundBone {boneName=dyn_fx_wide_a_r}
    componentLinearAccScale=<1.3 1.3 1.3>
    componentLinearVelScale=<1.3 1.3 1.3>
    gravityOverride {z=-100}
    physicsBodyDefinitions [{
      sphereCollisionRadius=0
      boundBone {boneName=dyn_fx_wide_a_r}
      boxExtents=<20 30 30>
      constraintSetup {angularLimitsMax=<15 0 15>, angularLimitsMin=<-15 -30 -15>, angularTarget=<1 0 0>}
      localJointOffset=<10 0 0>
     }]
   }
   n2 = Anim Dynamics - Bone: dyn_fx_wide_a_l(ComponentPose=@n1, Alpha=1, ExternalForce="0,0,0") {
    angularDampingOverride=1
    angularSpringConstant=20000
    bAngularSpring=true
    bGravityOverrideInSimSpace=true
    bOverrideAngularDamping=true
    bUseGravityOverride=true
    gravityScale=0.35
    simulationSpace=RootRelative
    alphaBoolBlend {blendOption=Linear}
    boundBone {boneName=dyn_fx_wide_a_l}
    componentLinearAccScale=<1.3 1.3 1.3>
    componentLinearVelScale=<1.3 1.3 1.3>
    gravityOverride {z=-100}
    physicsBodyDefinitions [{
      sphereCollisionRadius=0
      boundBone {boneName=dyn_fx_wide_a_l}
      boxExtents=<20 30 30>
      constraintSetup {angularLimitsMax=<15 0 15>, angularLimitsMin=<-15 -30 -15>, angularTarget=<1 0 0>}
      localJointOffset=<10 0 0>
     }]
   }
   n3 = Anim Dynamics - Bone: dyn_fx_wide_a(ComponentPose=@n2, Alpha=1, ExternalForce="0,0,0") {
    angularDampingOverride=1
    angularSpringConstant=20000
    bAngularSpring=true
    bGravityOverrideInSimSpace=true
    bOverrideAngularDamping=true
    bUseGravityOverride=true
    gravityScale=0.35
    simulationSpace=RootRelative
    alphaBoolBlend {blendOption=Linear}
    boundBone {boneName=dyn_fx_wide_a}
    componentLinearAccScale=<1.3 1.3 1.3>
    componentLinearVelScale=<1.3 1.3 1.3>
    gravityOverride {z=-100}
    physicsBodyDefinitions [{
      sphereCollisionRadius=0
      boundBone {boneName=dyn_fx_wide_a}
      boxExtents=<20 30 30>
      constraintSetup {angularLimitsMax=<15 0 15>, angularLimitsMin=<-15 -30 -15>, angularTarget=<1 0 0>}
      localJointOffset=<10 0 0>
     }]
   }
   n4 = Anim Dynamics - Bone: dyn_fx_wide_b_r(ComponentPose=@n3, Alpha=1, ExternalForce="0,0,0") {
    angularDampingOverride=1
    angularSpringConstant=20000
    bAngularSpring=true
    bGravityOverrideInSimSpace=true
    bOverrideAngularDamping=true
    bUseGravityOverride=true
    gravityScale=0.35
    simulationSpace=RootRelative
    alphaBoolBlend {blendOption=Linear}
    boundBone {boneName=dyn_fx_wide_b_r}
    componentLinearAccScale=<1.3 1.3 1.3>
    componentLinearVelScale=<1.3 1.3 1.3>
    gravityOverride {z=-100}
    physicsBodyDefinitions [{
      sphereCollisionRadius=0
      boundBone {boneName=dyn_fx_wide_b_r}
      boxExtents=<20 30 30>
      constraintSetup {angularLimitsMax=<15 0 15>, angularLimitsMin=<-15 -30 -15>, angularTarget=<1 0 0>}
      localJointOffset=<10 0 0>
     }]
   }
   n5 = Anim Dynamics - Bone: dyn_fx_wide_b_l(ComponentPose=@n4, Alpha=1, ExternalForce="0,0,0") {
    angularDampingOverride=1
    angularSpringConstant=20000
    bAngularSpring=true
    bGravityOverrideInSimSpace=true
    bOverrideAngularDamping=true
    bUseGravityOverride=true
    gravityScale=0.35
    simulationSpace=RootRelative
    alphaBoolBlend {blendOption=Linear}
    boundBone {boneName=dyn_fx_wide_b_l}
    componentLinearAccScale=<1.3 1.3 1.3>
    componentLinearVelScale=<1.3 1.3 1.3>
    gravityOverride {z=-100}
    physicsBodyDefinitions [{
      sphereCollisionRadius=0
      boundBone {boneName=dyn_fx_wide_b_l}
      boxExtents=<20 30 30>
      constraintSetup {angularLimitsMax=<15 0 15>, angularLimitsMin=<-15 -30 -15>, angularTarget=<1 0 0>}
      localJointOffset=<10 0 0>
     }]
   }
   n6 = Anim Dynamics - Bone: dyn_fx_wide_b(ComponentPose=@n5, Alpha=1, ExternalForce="0,0,0") {
    angularDampingOverride=1
    angularSpringConstant=20000
    bAngularSpring=true
    bGravityOverrideInSimSpace=true
    bOverrideAngularDamping=true
    bUseGravityOverride=true
    gravityScale=0.35
    simulationSpace=RootRelative
    alphaBoolBlend {blendOption=Linear}
    boundBone {boneName=dyn_fx_wide_b}
    componentLinearAccScale=<1.3 1.3 1.3>
    componentLinearVelScale=<1.3 1.3 1.3>
    gravityOverride {z=-100}
    physicsBodyDefinitions [{
      sphereCollisionRadius=0
      boundBone {boneName=dyn_fx_wide_b}
      boxExtents=<20 30 30>
      constraintSetup {angularLimitsMax=<15 0 15>, angularLimitsMin=<-15 -30 -15>, angularTarget=<1 0 0>}
      localJointOffset=<10 0 0>
     }]
   }
   n7 = Anim Dynamics - Bone: dyn_fx_wide_c_r(ComponentPose=@n6, Alpha=1, ExternalForce="0,0,0") {
    angularDampingOverride=1
    angularSpringConstant=20000
    bAngularSpring=true
    bGravityOverrideInSimSpace=true
    bOverrideAngularDamping=true
    bUseGravityOverride=true
    gravityScale=0.35
    simulationSpace=RootRelative
    alphaBoolBlend {blendOption=Linear}
    boundBone {boneName=dyn_fx_wide_c_r}
    componentLinearAccScale=<1.3 1.3 1.3>
    componentLinearVelScale=<1.3 1.3 1.3>
    gravityOverride {z=-100}
    physicsBodyDefinitions [{
      sphereCollisionRadius=0
      boundBone {boneName=dyn_fx_wide_c_r}
      boxExtents=<20 30 30>
      constraintSetup {angularLimitsMax=<15 0 15>, angularLimitsMin=<-15 -30 -15>, angularTarget=<1 0 0>}
      localJointOffset=<10 0 0>
     }]
   }
   n8 = Anim Dynamics - Bone: dyn_fx_wide_c_l(ComponentPose=@n7, Alpha=1, ExternalForce="0,0,0") {
    angularDampingOverride=1
    angularSpringConstant=20000
    bAngularSpring=true
    bGravityOverrideInSimSpace=true
    bOverrideAngularDamping=true
    bUseGravityOverride=true
    gravityScale=0.35
    simulationSpace=RootRelative
    alphaBoolBlend {blendOption=Linear}
    boundBone {boneName=dyn_fx_wide_c_l}
    componentLinearAccScale=<1.3 1.3 1.3>
    componentLinearVelScale=<1.3 1.3 1.3>
    gravityOverride {z=-100}
    physicsBodyDefinitions [{
      sphereCollisionRadius=0
      boundBone {boneName=dyn_fx_wide_c_l}
      boxExtents=<20 30 30>
      constraintSetup {angularLimitsMax=<15 0 15>, angularLimitsMin=<-15 -30 -15>, angularTarget=<1 0 0>}
      localJointOffset=<10 0 0>
     }]
   }
   n9 = Anim Dynamics - Bone: dyn_fx_wide_c(ComponentPose=@n8, Alpha=1, ExternalForce="0,0,0") {
    angularDampingOverride=1
    angularSpringConstant=20000
    bAngularSpring=true
    bGravityOverrideInSimSpace=true
    bOverrideAngularDamping=true
    bUseGravityOverride=true
    gravityScale=0.35
    simulationSpace=RootRelative
    alphaBoolBlend {blendOption=Linear}
    boundBone {boneName=dyn_fx_wide_c}
    componentLinearAccScale=<1.3 1.3 1.3>
    componentLinearVelScale=<1.3 1.3 1.3>
    gravityOverride {z=-100}
    physicsBodyDefinitions [{
      sphereCollisionRadius=0
      boundBone {boneName=dyn_fx_wide_c}
      boxExtents=<20 30 30>
      constraintSetup {angularLimitsMax=<15 0 15>, angularLimitsMin=<-15 -30 -15>, angularTarget=<1 0 0>}
      localJointOffset=<10 0 0>
     }]
   }
   n10 = Anim Dynamics - Bone: dyn_fx_narrow_b_r(ComponentPose=@n9, Alpha=1, ExternalForce="0,0,0") {
    angularDampingOverride=1
    angularSpringConstant=20000
    bAngularSpring=true
    bGravityOverrideInSimSpace=true
    bOverrideAngularDamping=true
    bUseGravityOverride=true
    gravityScale=0.35
    simulationSpace=RootRelative
    alphaBoolBlend {blendOption=Linear}
    boundBone {boneName=dyn_fx_narrow_b_r}
    componentLinearAccScale=<1.3 1.3 1.3>
    componentLinearVelScale=<1.3 1.3 1.3>
    gravityOverride {z=-100}
    physicsBodyDefinitions [{
      sphereCollisionRadius=0
      boundBone {boneName=dyn_fx_narrow_b_r}
      boxExtents=<20 30 30>
      constraintSetup {angularLimitsMax=<15 0 15>, angularLimitsMin=<-15 -45 -15>, angularTarget=<1 0 0>}
      localJointOffset=<5 0 0>
     }]
   }
   n11 = Anim Dynamics - Bone: dyn_fx_narrow_a_l(ComponentPose=@n10, Alpha=1, ExternalForce="0,0,0") {
    angularDampingOverride=1
    angularSpringConstant=20000
    bAngularSpring=true
    bGravityOverrideInSimSpace=true
    bOverrideAngularDamping=true
    bUseGravityOverride=true
    gravityScale=0.35
    simulationSpace=RootRelative
    alphaBoolBlend {blendOption=Linear}
    boundBone {boneName=dyn_fx_narrow_a_l}
    componentLinearAccScale=<1.3 1.3 1.3>
    componentLinearVelScale=<1.3 1.3 1.3>
    gravityOverride {z=-100}
    physicsBodyDefinitions [{
      sphereCollisionRadius=0
      boundBone {boneName=dyn_fx_narrow_a_l}
      boxExtents=<20 30 30>
      constraintSetup {angularLimitsMax=<15 0 15>, angularLimitsMin=<-15 -45 -15>, angularTarget=<1 0 0>}
      localJointOffset=<5 0 0>
     }]
   }
   n12 = Anim Dynamics - Bone: dyn_fx_narrow_a(ComponentPose=@n11, Alpha=1, ExternalForce="0,0,0") {
    angularDampingOverride=1
    angularSpringConstant=20000
    bAngularSpring=true
    bGravityOverrideInSimSpace=true
    bOverrideAngularDamping=true
    bUseGravityOverride=true
    gravityScale=0.35
    simulationSpace=RootRelative
    alphaBoolBlend {blendOption=Linear}
    boundBone {boneName=dyn_fx_narrow_a}
    componentLinearAccScale=<1.3 1.3 1.3>
    componentLinearVelScale=<1.3 1.3 1.3>
    gravityOverride {z=-100}
    physicsBodyDefinitions [{
      sphereCollisionRadius=0
      boundBone {boneName=dyn_fx_narrow_a}
      boxExtents=<20 30 30>
      constraintSetup {angularLimitsMax=<15 0 15>, angularLimitsMin=<-15 -45 -15>, angularTarget=<1 0 0>}
      localJointOffset=<5 0 0>
     }]
   }
   n13 = Anim Dynamics - Bone: dyn_fx_narrow_b_l(ComponentPose=@n12, Alpha=1, ExternalForce="0,0,0") {
    angularDampingOverride=1
    angularSpringConstant=20000
    bAngularSpring=true
    bGravityOverrideInSimSpace=true
    bOverrideAngularDamping=true
    bUseGravityOverride=true
    gravityScale=0.35
    simulationSpace=RootRelative
    alphaBoolBlend {blendOption=Linear}
    boundBone {boneName=dyn_fx_narrow_b_l}
    componentLinearAccScale=<1.3 1.3 1.3>
    componentLinearVelScale=<1.3 1.3 1.3>
    gravityOverride {z=-100}
    physicsBodyDefinitions [{
      sphereCollisionRadius=0
      boundBone {boneName=dyn_fx_narrow_b_l}
      boxExtents=<20 30 30>
      constraintSetup {angularLimitsMax=<15 0 15>, angularLimitsMin=<-15 -45 -15>, angularTarget=<1 0 0>}
      localJointOffset=<5 0 0>
     }]
   }
   n14 = Anim Dynamics - Bone: dyn_fx_narrow_b(ComponentPose=@n13, Alpha=1, ExternalForce="0,0,0") {
    angularDampingOverride=1
    angularSpringConstant=20000
    bAngularSpring=true
    bGravityOverrideInSimSpace=true
    bOverrideAngularDamping=true
    bUseGravityOverride=true
    gravityScale=0.35
    simulationSpace=RootRelative
    alphaBoolBlend {blendOption=Linear}
    boundBone {boneName=dyn_fx_narrow_b}
    componentLinearAccScale=<1.3 1.3 1.3>
    componentLinearVelScale=<1.3 1.3 1.3>
    gravityOverride {z=-100}
    physicsBodyDefinitions [{
      sphereCollisionRadius=0
      boundBone {boneName=dyn_fx_narrow_b}
      boxExtents=<20 30 30>
      constraintSetup {angularLimitsMax=<15 0 15>, angularLimitsMin=<-15 -45 -15>, angularTarget=<1 0 0>}
      localJointOffset=<5 0 0>
     }]
   }
   n15 = Anim Dynamics - Bone: dyn_fx_narrow_a_r(ComponentPose=@n14, Alpha=1, ExternalForce="0,0,0") {
    angularDampingOverride=1
    angularSpringConstant=20000
    bAngularSpring=true
    bGravityOverrideInSimSpace=true
    bOverrideAngularDamping=true
    bUseGravityOverride=true
    gravityScale=0.35
    simulationSpace=RootRelative
    alphaBoolBlend {blendOption=Linear}
    boundBone {boneName=dyn_fx_narrow_a_r}
    componentLinearAccScale=<1.3 1.3 1.3>
    componentLinearVelScale=<1.3 1.3 1.3>
    gravityOverride {z=-100}
    physicsBodyDefinitions [{
      sphereCollisionRadius=0
      boundBone {boneName=dyn_fx_narrow_a_r}
      boxExtents=<20 30 30>
      constraintSetup {angularLimitsMax=<15 0 15>, angularLimitsMin=<-15 -45 -15>, angularTarget=<1 0 0>}
      localJointOffset=<5 0 0>
     }]
   }
   n16 = Anim Dynamics - Bone: dyn_fx_narrow_c(ComponentPose=@n15, Alpha=1, ExternalForce="0,0,0") {
    angularDampingOverride=1
    angularSpringConstant=20000
    bAngularSpring=true
    bGravityOverrideInSimSpace=true
    bOverrideAngularDamping=true
    bUseGravityOverride=true
    gravityScale=0.35
    simulationSpace=RootRelative
    alphaBoolBlend {blendOption=Linear}
    boundBone {boneName=dyn_fx_narrow_c}
    componentLinearAccScale=<1.3 1.3 1.3>
    componentLinearVelScale=<1.3 1.3 1.3>
    gravityOverride {z=-100}
    physicsBodyDefinitions [{
      sphereCollisionRadius=0
      boundBone {boneName=dyn_fx_narrow_c}
      boxExtents=<20 30 30>
      constraintSetup {angularLimitsMax=<15 0 15>, angularLimitsMin=<-15 -45 -15>, angularTarget=<1 0 0>}
      localJointOffset=<5 0 0>
     }]
   }
   n17 = Get bIsInVehicle() -> (bIsInVehicle:bool)
   n18 = "Transform (Modify) Bone - Bone: dyn_fx_wide_a_l"!(ComponentPose=@n16, bAlphaBoolEnabled:bool=@n17, Translation="0,0,0", Rotation="0,0,0", Scale="0.01,0.01,0.01") -> (Pose) {
    alphaInputType=Bool
    rotationMode=BMM_Additive
    rotationSpace=BCS_BoneSpace
    scaleMode=BMM_Additive
    scaleSpace=BCS_BoneSpace
    translationMode=BMM_Additive
    translationSpace=BCS_BoneSpace
    alphaBoolBlend {blendOption=Linear}
    boneToModify {boneName=dyn_fx_wide_a_l}
    scale=<0.01 0.01 0.01>
   }
   n19 = "Transform (Modify) Bone - Bone: dyn_fx_wide_a_r"!(ComponentPose=@n18, bAlphaBoolEnabled:bool=@n17, Translation="0,0,0", Rotation="0,0,0", Scale="0.01,0.01,0.01") -> (Pose) {
    alphaInputType=Bool
    rotationMode=BMM_Additive
    rotationSpace=BCS_BoneSpace
    scaleMode=BMM_Additive
    scaleSpace=BCS_BoneSpace
    translationMode=BMM_Additive
    translationSpace=BCS_BoneSpace
    alphaBoolBlend {blendOption=Linear}
    boneToModify {boneName=dyn_fx_wide_a_r}
    scale=<0.01 0.01 0.01>
   }
   n20 = "Transform (Modify) Bone - Bone: dyn_fx_wide_a"!(ComponentPose=@n19, bAlphaBoolEnabled:bool=@n17, Translation="0,0,0", Rotation="0,0,0", Scale="0.01,0.01,0.01") -> (Pose) {
    alphaInputType=Bool
    rotationMode=BMM_Additive
    rotationSpace=BCS_BoneSpace
    scaleMode=BMM_Additive
    scaleSpace=BCS_BoneSpace
    translationMode=BMM_Additive
    translationSpace=BCS_BoneSpace
    alphaBoolBlend {blendOption=Linear}
    boneToModify {boneName=dyn_fx_wide_a}
    scale=<0.01 0.01 0.01>
   }
   n21 = Get bIsInVehicle()
   n22 = "Transform (Modify) Bone - Bone: dyn_fx_narrow_a_l"!(ComponentPose=@n20, bAlphaBoolEnabled:bool=@n21, Translation="0,0,0", Rotation="0,0,0", Scale="0.01,0.01,0.01") -> (Pose) {
    alphaInputType=Bool
    rotationMode=BMM_Additive
    rotationSpace=BCS_BoneSpace
    scaleMode=BMM_Additive
    scaleSpace=BCS_BoneSpace
    translationMode=BMM_Additive
    translationSpace=BCS_BoneSpace
    alphaBoolBlend {blendOption=Linear}
    boneToModify {boneName=dyn_fx_narrow_a_l}
    scale=<0.01 0.01 0.01>
   }
   n23 = "Transform (Modify) Bone - Bone: dyn_fx_narrow_a_r"!(ComponentPose=@n22, bAlphaBoolEnabled:bool=@n21, Translation="0,0,0", Rotation="0,0,0", Scale="0.01,0.01,0.01") -> (Pose) {
    alphaInputType=Bool
    rotationMode=BMM_Additive
    rotationSpace=BCS_BoneSpace
    scaleMode=BMM_Additive
    scaleSpace=BCS_BoneSpace
    translationMode=BMM_Additive
    translationSpace=BCS_BoneSpace
    alphaBoolBlend {blendOption=Linear}
    boneToModify {boneName=dyn_fx_narrow_a_r}
    scale=<0.01 0.01 0.01>
   }
   n24 = "Transform (Modify) Bone - Bone: dyn_fx_narrow_a"!(ComponentPose=@n23, bAlphaBoolEnabled:bool=@n21, Translation="0,0,0", Rotation="0,0,0", Scale="0.01,0.01,0.01") -> (Pose) {
    alphaInputType=Bool
    rotationMode=BMM_Additive
    rotationSpace=BCS_BoneSpace
    scaleMode=BMM_Additive
    scaleSpace=BCS_BoneSpace
    translationMode=BMM_Additive
    translationSpace=BCS_BoneSpace
    alphaBoolBlend {blendOption=Linear}
    boneToModify {boneName=dyn_fx_narrow_a}
    scale=<0.01 0.01 0.01>
   }
   Output Pose(Result=Component To Local(ComponentPose=@n24)) {layerGroup=None}
   n25 = Break Pose Deformation Query Results Arms!(PoseDeformationQueryResults_Arms:"Pose Deformation Query Results Arms Structure (by ref)"=Break Pose Deformation Query Arms!(PoseDeformationQuery_Arms:"Pose Deformation Query Arms Structure (by ref)"=Get Arms() -> (Arms:Pose Deformation Query Arms Structure)) -> (QueryResults:Pose Deformation Query Results Arms Structure) {is=title K2Node_BreakStruct}) -> (
    ArmUpL:float=~
    ArmUpR:float=~
    ArmForwardL:float=~
    ArmForwardR:float=~
    ArmBackL:float=~
    ArmBackR:float=~
    ArmForwardUpL:float=~
    ArmForwardUpR:float=~
    ArmForwardDownL:float=~
    ArmForwardDownR:float=~
   ) {is=title K2Node_BreakStruct}
   n26 = Control Rig!(
    Source=Copy Pose From Mesh() {bUseAttachedParent=true}
    Alpha
    thigh_rot_l:double=Break Buffet Pose Space Deformer Input!(BuffetPoseSpaceDeformerInput:"Buffet Pose Space Deformer Input Structure (by ref)"=Get ThighRotL() -> (ThighRotL:Buffet Pose Space Deformer Input Structure)) -> (DeformationValue:float=~) {is=title K2Node_BreakStruct}
    elbow_rot_l:double=Break Buffet Pose Space Deformer Input(BuffetPoseSpaceDeformerInput=Get ElbowRotL() -> (ElbowRotL:Buffet Pose Space Deformer Input Structure)) -> (DeformationValue=~)
    elbow_rot_r:double=Break Buffet Pose Space Deformer Input(BuffetPoseSpaceDeformerInput=Get ElbowRotR() -> (ElbowRotR:Buffet Pose Space Deformer Input Structure)) -> (DeformationValue=~)
    thigh_rot_r:double=Break Buffet Pose Space Deformer Input(BuffetPoseSpaceDeformerInput=Get ThighRotR() -> (ThighRotR:Buffet Pose Space Deformer Input Structure)) -> (DeformationValue=~)
    knee_rot_l:double=Break Buffet Pose Space Deformer Input(BuffetPoseSpaceDeformerInput=Get KneeRotL() -> (KneeRotL:Buffet Pose Space Deformer Input Structure)) -> (DeformationValue=~)
    knee_rot_r:double=Break Buffet Pose Space Deformer Input(BuffetPoseSpaceDeformerInput=Get KneeRotR() -> (KneeRotR:Buffet Pose Space Deformer Input Structure)) -> (DeformationValue=~)
    psd_arm_up_alpha_l:double=@n25
    psd_arm_up_alpha_r:double=@n25.ArmUpR
    psd_arm_fwd_alpha_l:double=@n25.ArmForwardL
    psd_arm_fwd_alpha_r:double=@n25.ArmForwardR
    psd_arm_bk_alpha_l:double=@n25.ArmBackL
    psd_arm_bk_alpha_r:double=@n25.ArmBackR
    psd_arm_fwd_up_alpha_l:double=@n25.ArmForwardUpL
    psd_arm_fwd_up_alpha_r:double=@n25.ArmForwardUpR
    psd_arm_fwd_dn_alpha_l:double=@n25.ArmForwardDownL
    psd_arm_fwd_dn_alpha_r:double=@n25.ArmForwardDownR
   ) -> (Pose) {
    alphaBoolBlend {blendOption=Linear}
    controlRigAssetReference {blueprintRigClass=&/F_MED_Muscle_Volume_ControlRig.F_MED_Muscle_Volume_ControlRig_C}
   }
   n27 = "Transform (Modify) Bone - Bone: simspace"(
    ComponentPose=Local To Component(LocalPose=@n26)
    Alpha=1
    Translation="0,0,0"
    Rotation=Get sim_space_rotation_$() -> (sim_space_rotation_$:Rotator)
    Scale="1,1,1"
   ) {
    rotationMode=BMM_Replace
    rotationSpace=BCS_WorldSpace
    translationSpace=BCS_BoneSpace
    alphaBoolBlend {blendOption=Linear}
    alphaScaleBiasClamp {bInterpResult=true, interpSpeedDecreasing=2, interpSpeedIncreasing=2}
    boneToModify {boneName=simspace}
   }
   n28 = "Constraint - simspace (P)"!(ComponentPose=@n27, bAlphaBoolEnabled:bool=Get bIsInFrontEnd(), ConstraintWeights_0=1) -> (Pose) {
    alphaInputType=Bool
    alphaBoolBlend {blendOption=Linear}
    boneToModify {boneName=simspace}
    constraintSetup [{targetBone {boneName=pelvis}}]
    constraintWeights [1]
   }
   Save cached pose 'deformRigPose'(Pose=Component To Local(ComponentPose=@n28)) {cacheName=deformRigPose}
   n29 = "Constraint - dyn_shard_shoulder_a_base_l (PPPP)"!(
    ComponentPose=Local To Component(LocalPose=Use cached pose 'SkirtCachePose'())
    Alpha=1
    ConstraintWeights_0=0.45
    ConstraintWeights_1:float=0.2
    ConstraintWeights_2:float=0.1
    ConstraintWeights_3:float=0.25
   ) -> (Pose) {
    alphaBoolBlend {blendOption=Linear}
    boneToModify {boneName=dyn_shard_shoulder_a_base_l}
    constraintSetup [
     {targetBone {boneName=upperarm_twist_01_l}}
     {targetBone {boneName=spine_05}}
     {targetBone {boneName=clavicle_l}}
     {targetBone {boneName=upperarm_fr_volume_l}}
    ]
    constraintWeights [1,1,1,1]
   }
   n30 = "Constraint - dyn_shard_shoulder_b_base_l (PPPPP)"!(
    ComponentPose=@n29
    Alpha=1
    ConstraintWeights_0=0.2
    ConstraintWeights_1:float=0.45
    ConstraintWeights_2:float=0.2
    ConstraintWeights_3:float=0
    ConstraintWeights_4:float=0.15
   ) -> (Pose) {
    alphaBoolBlend {blendOption=Linear}
    boneToModify {boneName=dyn_shard_shoulder_b_base_l}
    constraintSetup [
     {targetBone {boneName=upperarm_up_volume_l}}
     {targetBone {boneName=clavicle_l}}
     {targetBone {boneName=upperarm_twist_01_l}}
     {targetBone {boneName=upperarm_fr_volume_l}}
     {targetBone {boneName=spine_05}}
    ]
    constraintWeights [1,1,1,1,1]
   }
   n31 = "Constraint - dyn_shard_shoulder_c_base_l (PPP)"!(ComponentPose=@n30, Alpha=1, ConstraintWeights_0=0.95, ConstraintWeights_1:float=0.04, ConstraintWeights_2:float=0.01) -> (Pose) {
    alphaBoolBlend {blendOption=Linear}
    boneToModify {boneName=dyn_shard_shoulder_c_base_l}
    constraintSetup [{targetBone {boneName=upperarm_twist_01_l}}, {targetBone {boneName=upperarm_up_volume_l}}, {targetBone {boneName=upperarm_fr_volume_l}}]
    constraintWeights [1,1,1]
   }
   n32 = "Constraint - dyn_shard_shoulder_d_base_l (PPP)"!(ComponentPose=@n31, Alpha=1, ConstraintWeights_0=0.1, ConstraintWeights_1:float=0.09, ConstraintWeights_2:float=0.81) -> (Pose) {
    alphaBoolBlend {blendOption=Linear}
    boneToModify {boneName=dyn_shard_shoulder_d_base_l}
    constraintSetup [{targetBone {boneName=upperarm_twist_02_l}}, {targetBone {boneName=upperarm_l}}, {targetBone {boneName=upperarm_twist_01_l}}]
    constraintWeights [1,1,1]
   }
   n33 = "Constraint - dyn_shard_shoulder_e_base_l (PP)"!(ComponentPose=@n32, Alpha=1, ConstraintWeights_0=0.2, ConstraintWeights_1:float=0.8) -> (Pose) {
    alphaBoolBlend {blendOption=Linear}
    boneToModify {boneName=dyn_shard_shoulder_e_base_l}
    constraintSetup [{targetBone {boneName=clavicle_l}}, {targetBone {boneName=upperarm_twist_01_l}}]
    constraintWeights [1,1]
   }
   n34 = "Constraint - dyn_shard_shoulder_f_base_l (PPPP)"!(
    ComponentPose=@n33
    Alpha=1
    ConstraintWeights_0=0.16
    ConstraintWeights_1:float=0.21
    ConstraintWeights_2:float=0.43
    ConstraintWeights_3:float=0.2
   ) -> (Pose) {
    alphaBoolBlend {blendOption=Linear}
    boneToModify {boneName=dyn_shard_shoulder_f_base_l}
    constraintSetup [
     {targetBone {boneName=upperarm_bk_volume_l}}
     {targetBone {boneName=clavicle_l}}
     {targetBone {boneName=upperarm_twist_01_l}}
     {targetBone {boneName=upperarm_up_volume_l}}
    ]
    constraintWeights [1,1,1,1]
   }
   n35 = "Constraint - dyn_shard_shoulder_g_base_l (PPPP)"!(
    ComponentPose=@n34
    Alpha=1
    ConstraintWeights_0=0.15
    ConstraintWeights_1:float=0.45
    ConstraintWeights_2:float=0.1
    ConstraintWeights_3:float=0.3
   ) -> (Pose) {
    alphaBoolBlend {blendOption=Linear}
    boneToModify {boneName=dyn_shard_shoulder_g_base_l}
    constraintSetup [
     {targetBone {boneName=upperarm_bk_volume_l}}
     {targetBone {boneName=clavicle_l}}
     {targetBone {boneName=spine_05}}
     {targetBone {boneName=upperarm_twist_01_l}}
    ]
    constraintWeights [1,1,1,1]
   }
   n36 = "Constraint - dyn_shard_shoulder_h_base_l (PPPP)"!(
    ComponentPose=@n35
    Alpha=1
    ConstraintWeights_0=0.26
    ConstraintWeights_1:float=0.22
    ConstraintWeights_2:float=0.38
    ConstraintWeights_3:float=0.14
   ) -> (Pose) {
    alphaBoolBlend {blendOption=Linear}
    boneToModify {boneName=dyn_shard_shoulder_h_base_l}
    constraintSetup [
     {targetBone {boneName=upperarm_bk_volume_l}}
     {targetBone {boneName=clavicle_l}}
     {targetBone {boneName=upperarm_twist_01_l}}
     {targetBone {boneName=upperarm_up_volume_l}}
    ]
    constraintWeights [1,1,1,1]
   }
   n37 = "Constraint - dyn_shard_skirt_a_base_l (PPP)"!(ComponentPose=@n36, Alpha=1, ConstraintWeights_0=0.02, ConstraintWeights_1:float=0.81, ConstraintWeights_2:float=0.17) -> (Pose) {
    alphaBoolBlend {blendOption=Linear}
    boneToModify {boneName=dyn_shard_skirt_a_base_l}
    constraintSetup [{targetBone {boneName=dyn_skirt_side_b_l}}, {targetBone {boneName=dyn_skirt_fr_b_l}}, {targetBone {boneName=dyn_skirt_fr_b}}]
    constraintWeights [1,1,1]
   }
   n38 = "Constraint - dyn_shard_skirt_b_base_l (PPP)"!(ComponentPose=@n37, Alpha=1, ConstraintWeights_0=0.12, ConstraintWeights_1:float=0.84, ConstraintWeights_2:float=0.04) -> (Pose) {
    alphaBoolBlend {blendOption=Linear}
    boneToModify {boneName=dyn_shard_skirt_b_base_l}
    constraintSetup [{targetBone {boneName=dyn_skirt_side_b_l}}, {targetBone {boneName=dyn_skirt_fr_b_l}}, {targetBone {boneName=dyn_skirt_fr_b}}]
    constraintWeights [1,1,1]
   }
   n39 = "Constraint - dyn_shard_skirt_c_base_l (PP)"!(ComponentPose=@n38, Alpha=1, ConstraintWeights_0=0.38, ConstraintWeights_1:float=0.62) -> (Pose) {
    alphaBoolBlend {blendOption=Linear}
    boneToModify {boneName=dyn_shard_skirt_c_base_l}
    constraintSetup [{targetBone {boneName=dyn_skirt_side_b_l}}, {targetBone {boneName=dyn_skirt_fr_b_l}}]
    constraintWeights [1,1]
   }
   n40 = "Constraint - dyn_shard_skirt_d_base_l (PPP)"!(ComponentPose=@n39, Alpha=1, ConstraintWeights_0=0.77, ConstraintWeights_1:float=0.14, ConstraintWeights_2:float=0.09) -> (Pose) {
    alphaBoolBlend {blendOption=Linear}
    boneToModify {boneName=dyn_shard_skirt_d_base_l}
    constraintSetup [{targetBone {boneName=dyn_skirt_side_b_l}}, {targetBone {boneName=dyn_skirt_fr_b_l}}, {targetBone {boneName=dyn_skirt_bk_b_l}}]
    constraintWeights [1,1,1]
   }
   n41 = "Constraint - dyn_shard_skirt_e_base_l (PPP)"!(ComponentPose=@n40, Alpha=1, ConstraintWeights_0=0.75, ConstraintWeights_1:float=0.06, ConstraintWeights_2:float=0.19) -> (Pose) {
    alphaBoolBlend {blendOption=Linear}
    boneToModify {boneName=dyn_shard_skirt_e_base_l}
    constraintSetup [{targetBone {boneName=dyn_skirt_side_b_l}}, {targetBone {boneName=dyn_skirt_fr_b_l}}, {targetBone {boneName=dyn_skirt_bk_b_l}}]
    constraintWeights [1,1,1]
   }
   n42 = "Constraint - dyn_shard_skirt_a_base_r (PPP)"!(ComponentPose=@n41, Alpha=1, ConstraintWeights_0=0.34, ConstraintWeights_1:float=0.01, ConstraintWeights_2:float=0.65) -> (Pose) {
    alphaBoolBlend {blendOption=Linear}
    boneToModify {boneName=dyn_shard_skirt_a_base_r}
    constraintSetup [{targetBone {boneName=dyn_skirt_fr_b}}, {targetBone {boneName=dyn_skirt_fr_b_l}}, {targetBone {boneName=dyn_skirt_fr_b_r}}]
    constraintWeights [1,1,1]
   }
   n43 = "Constraint - dyn_shard_skirt_b_base_r (PP)"!(ComponentPose=@n42, Alpha=1, ConstraintWeights_0=0.9, ConstraintWeights_1:float=0.1) -> (Pose) {
    alphaBoolBlend {blendOption=Linear}
    boneToModify {boneName=dyn_shard_skirt_b_base_r}
    constraintSetup [{targetBone {boneName=dyn_skirt_fr_c_r}}, {targetBone {boneName=dyn_skirt_side_c_r}}]
    constraintWeights [1,1]
   }
   n44 = "Constraint - dyn_shard_skirt_c_base_r (PP)"!(ComponentPose=@n43, Alpha=1, ConstraintWeights_0=0.9, ConstraintWeights_1:float=0.1) -> (Pose) {
    alphaBoolBlend {blendOption=Linear}
    boneToModify {boneName=dyn_shard_skirt_c_base_r}
    constraintSetup [{targetBone {boneName=dyn_skirt_fr_c_r}}, {targetBone {boneName=dyn_skirt_side_c_r}}]
    constraintWeights [1,1]
   }
   n45 = "Constraint - dyn_shard_skirt_d_base_r (PP)"!(ComponentPose=@n44, Alpha=1, ConstraintWeights_0=0.76, ConstraintWeights_1:float=0.24) -> (Pose) {
    alphaBoolBlend {blendOption=Linear}
    boneToModify {boneName=dyn_shard_skirt_d_base_r}
    constraintSetup [{targetBone {boneName=dyn_skirt_fr_b_r}}, {targetBone {boneName=dyn_skirt_side_b_r}}]
    constraintWeights [1,1]
   }
   n46 = "Constraint - dyn_shard_skirt_e_base_r (PP)"!(ComponentPose=@n45, Alpha=1, ConstraintWeights_0=0.5, ConstraintWeights_1:float=0.5) -> (Pose) {
    alphaBoolBlend {blendOption=Linear}
    boneToModify {boneName=dyn_shard_skirt_e_base_r}
    constraintSetup [{targetBone {boneName=dyn_skirt_fr_c_r}}, {targetBone {boneName=dyn_skirt_side_c_r}}]
    constraintWeights [1,1]
   }
   n47 = "Constraint - dyn_shard_skirt_f_base_r (PP)"!(ComponentPose=@n46, Alpha=1, ConstraintWeights_0=0.3, ConstraintWeights_1:float=0.7) -> (Pose) {
    alphaBoolBlend {blendOption=Linear}
    boneToModify {boneName=dyn_shard_skirt_f_base_r}
    constraintSetup [{targetBone {boneName=dyn_skirt_fr_c_r}}, {targetBone {boneName=dyn_skirt_side_c_r}}]
    constraintWeights [1,1]
   }
   n48 = "Constraint - dyn_shard_skirt_g_base_r (PPPP)"!(
    ComponentPose=@n47
    Alpha=1
    ConstraintWeights_0=0.03
    ConstraintWeights_1:float=0.03
    ConstraintWeights_2:float=0.41
    ConstraintWeights_3:float=0.53
   ) -> (Pose) {
    alphaBoolBlend {blendOption=Linear}
    boneToModify {boneName=dyn_shard_skirt_g_base_r}
    constraintSetup [
     {targetBone {boneName=dyn_skirt_side_a_r}}
     {targetBone {boneName=dyn_skirt_fr_a_r}}
     {targetBone {boneName=dyn_skirt_fr_b_r}}
     {targetBone {boneName=dyn_skirt_side_b_r}}
    ]
    constraintWeights [1,1,1,1]
   }
   n49 = "Constraint - dyn_shard_skirt_h_base_r (PP)"!(ComponentPose=@n48, Alpha=1, ConstraintWeights_0=0.7, ConstraintWeights_1:float=0.3) -> (Pose) {
    alphaBoolBlend {blendOption=Linear}
    boneToModify {boneName=dyn_shard_skirt_h_base_r}
    constraintSetup [{targetBone {boneName=dyn_skirt_bk_c_r}}, {targetBone {boneName=dyn_skirt_side_c_r}}]
    constraintWeights [1,1]
   }
   n50 = "Constraint - dyn_shard_skirt_i_base_r (PPPPPP)"!(
    ComponentPose=@n49
    Alpha=1
    ConstraintWeights_0=0.06
    ConstraintWeights_1:float=0.38
    ConstraintWeights_2:float=0.06
    ConstraintWeights_3:float=0.06
    ConstraintWeights_4:float=0.38
    ConstraintWeights_5:float=0.06
   ) -> (Pose) {
    alphaBoolBlend {blendOption=Linear}
    boneToModify {boneName=dyn_shard_skirt_i_base_r}
    constraintSetup [
     {targetBone {boneName=dyn_skirt_side_a_r}}
     {targetBone {boneName=dyn_skirt_bk_b_r}}
     {targetBone {boneName=dyn_skirt_bk_b}}
     {targetBone {boneName=dyn_skirt_bk_a}}
     {targetBone {boneName=dyn_skirt_bk_a_r}}
     {targetBone {boneName=dyn_skirt_side_b_r}}
    ]
    constraintWeights [1,1,1,1,1,1]
   }
   n51 = "Constraint - dyn_shard_skirt_j_base_r (PPPP)"!(
    ComponentPose=@n50
    Alpha=1
    ConstraintWeights_0=0.7
    ConstraintWeights_1:float=0.15
    ConstraintWeights_2:float=0.1
    ConstraintWeights_3:float=0.05
   ) -> (Pose) {
    alphaBoolBlend {blendOption=Linear}
    boneToModify {boneName=dyn_shard_skirt_j_base_r}
    constraintSetup [
     {targetBone {boneName=dyn_skirt_side_a_r}}
     {targetBone {boneName=pelvis}}
     {targetBone {boneName=dyn_skirt_fr_a_r}}
     {targetBone {boneName=dyn_skirt_bk_a_r}}
    ]
    constraintWeights [1,1,1,1]
   }
   n52 = Get bIsSkyDiving()
   n53 = "Transform (Modify) Bone - Bone: dyn_shard_shoulder_a_base_l"!(ComponentPose=@n51, bAlphaBoolEnabled:bool=@n52, Translation="0,0,2", Rotation="-10,0,0", Scale="1,1,1") -> (Pose) {
    alphaInputType=Bool
    rotationMode=BMM_Additive
    rotationSpace=BCS_BoneSpace
    translationMode=BMM_Additive
    translationSpace=BCS_BoneSpace
    alphaBoolBlend {blendOption=Linear}
    boneToModify {boneName=dyn_shard_shoulder_a_base_l}
    rotation {pitch=-10}
    translation {z=2}
   }
   n54 = "Transform (Modify) Bone - Bone: dyn_shard_shoulder_b_base_l"!(ComponentPose=@n53, bAlphaBoolEnabled:bool=@n52, Translation="0,0,3", Rotation="-15,0,0", Scale="1,1,1") -> (Pose) {
    alphaInputType=Bool
    rotationMode=BMM_Additive
    rotationSpace=BCS_BoneSpace
    translationMode=BMM_Additive
    translationSpace=BCS_BoneSpace
    alphaBoolBlend {blendOption=Linear}
    boneToModify {boneName=dyn_shard_shoulder_b_base_l}
    rotation {pitch=-15}
    translation {z=3}
   }
   n55 = Anim Dynamics - Bone: dyn_shard_skirt_a_base_l(ComponentPose=@n54, Alpha=0.3, ExternalForce="0,0,0") {
    angularDampingOverride=0.999
    angularSpringConstant=150
    bAngularSpring=true
    bGravityOverrideInSimSpace=true
    bLinearSpring=true
    bOverrideAngularDamping=true
    bOverrideLinearDamping=true
    bUseGravityOverride=true
    gravityScale=0
    linearDampingOverride=0.995
    linearSpringConstant=100
    simulationSpace=RootRelative
    alphaBoolBlend {blendOption=Linear}
    alphaScaleBiasClamp {interpSpeedDecreasing=3, interpSpeedIncreasing=5}
    boundBone {boneName=dyn_shard_skirt_a_base_l}
    componentAppliedLinearAccClamp=<6000 6000 6000>
    componentLinearAccScale=<0.05 0.05 0.05>
    componentLinearVelScale=<0.8 0.8 0.6>
    gravityOverride=<0.25 0.25 0.25>
    physicsBodyDefinitions [{
      sphereCollisionRadius=0
      boundBone {boneName=dyn_shard_skirt_a_base_l}
      boxExtents=<50 50 50>
      constraintSetup {angularLimitsMax=<0 0 0>, angularLimitsMin=<0 0 0>, angularTarget=<0 0 0>, linearAxesMax=<1 0.5 0>, linearAxesMin=<-0.5 -0.5 -0.5>}
      localJointOffset=<-3 0 0>
     }]
    relativeSpaceBone {boneName=root}
   }
   n56 = Anim Dynamics - Bone: dyn_shard_skirt_b_base_l(ComponentPose=@n55, Alpha=0.75, ExternalForce="0,0,0") {
    angularDampingOverride=0.999
    angularSpringConstant=150
    bAngularSpring=true
    bGravityOverrideInSimSpace=true
    bLinearSpring=true
    bOverrideAngularDamping=true
    bOverrideLinearDamping=true
    bUseGravityOverride=true
    gravityScale=0
    linearDampingOverride=0.995
    linearSpringConstant=100
    simulationSpace=RootRelative
    alphaBoolBlend {blendOption=Linear}
    alphaScaleBiasClamp {interpSpeedDecreasing=3, interpSpeedIncreasing=5}
    boundBone {boneName=dyn_shard_skirt_b_base_l}
    componentAppliedLinearAccClamp=<6000 6000 6000>
    componentLinearAccScale=<0.05 0.05 0.05>
    componentLinearVelScale=<0.8 0.8 0.6>
    gravityOverride=<0.25 0.25 0.25>
    physicsBodyDefinitions [{
      sphereCollisionRadius=0
      boundBone {boneName=dyn_shard_skirt_b_base_l}
      boxExtents=<50 50 50>
      constraintSetup {angularLimitsMax=<0 0 0>, angularLimitsMin=<0 0 0>, angularTarget=<0 0 0>, linearAxesMax=<1 0.5 0>, linearAxesMin=<-0.5 -0.5 -0.5>}
      localJointOffset=<-3 0 0>
     }]
    relativeSpaceBone {boneName=root}
   }
   n57 = Anim Dynamics - Bone: dyn_shard_skirt_c_base_l(ComponentPose=@n56, Alpha=0.7, ExternalForce="0,0,0") {
    angularDampingOverride=0.999
    angularSpringConstant=150
    bAngularSpring=true
    bGravityOverrideInSimSpace=true
    bLinearSpring=true
    bOverrideAngularDamping=true
    bOverrideLinearDamping=true
    bUseGravityOverride=true
    gravityScale=0
    linearDampingOverride=0.995
    linearSpringConstant=100
    simulationSpace=RootRelative
    alphaBoolBlend {blendOption=Linear}
    alphaScaleBiasClamp {interpSpeedDecreasing=3, interpSpeedIncreasing=5}
    boundBone {boneName=dyn_shard_skirt_c_base_l}
    componentAppliedLinearAccClamp=<6000 6000 6000>
    componentLinearAccScale=<0.05 0.05 0.05>
    componentLinearVelScale=<0.8 0.8 0.6>
    gravityOverride=<0.25 0.25 0.25>
    physicsBodyDefinitions [{
      sphereCollisionRadius=0
      boundBone {boneName=dyn_shard_skirt_c_base_l}
      boxExtents=<50 50 50>
      constraintSetup {angularLimitsMax=<0 0 0>, angularLimitsMin=<0 0 0>, angularTarget=<0 0 0>, linearAxesMax=<1 0.5 0>, linearAxesMin=<0 -0.5 -1>}
      localJointOffset=<-3 0 0>
     }]
    relativeSpaceBone {boneName=root}
   }
   n58 = Anim Dynamics - Bone: dyn_shard_skirt_d_base_l(ComponentPose=@n57, Alpha=0.7, ExternalForce="0,0,0") {
    angularDampingOverride=0.999
    angularSpringConstant=150
    bAngularSpring=true
    bGravityOverrideInSimSpace=true
    bLinearSpring=true
    bOverrideAngularDamping=true
    bOverrideLinearDamping=true
    bUseGravityOverride=true
    gravityScale=0
    linearDampingOverride=0.995
    linearSpringConstant=100
    simulationSpace=RootRelative
    alphaBoolBlend {blendOption=Linear}
    alphaScaleBiasClamp {interpSpeedDecreasing=3, interpSpeedIncreasing=5}
    boundBone {boneName=dyn_shard_skirt_d_base_l}
    componentAppliedLinearAccClamp=<6000 6000 6000>
    componentLinearAccScale=<0.05 0.05 0.05>
    componentLinearVelScale=<0.8 0.8 0.6>
    gravityOverride=<0.25 0.25 0.25>
    physicsBodyDefinitions [{
      sphereCollisionRadius=0
      boundBone {boneName=dyn_shard_skirt_d_base_l}
      boxExtents=<50 50 50>
      constraintSetup {angularLimitsMax=<0 0 0>, angularLimitsMin=<0 0 0>, angularTarget=<0 0 0>, linearAxesMax=<1 0.5 0>, linearAxesMin=<-1 -0.5 -1>}
      localJointOffset=<-3 0 0>
     }]
    relativeSpaceBone {boneName=root}
   }
   n59 = Anim Dynamics - Bone: dyn_shard_skirt_e_base_l(ComponentPose=@n58, Alpha=0.5, ExternalForce="0,0,0") {
    angularDampingOverride=0.999
    angularSpringConstant=150
    bAngularSpring=true
    bGravityOverrideInSimSpace=true
    bLinearSpring=true
    bOverrideAngularDamping=true
    bOverrideLinearDamping=true
    bUseGravityOverride=true
    gravityScale=0
    linearDampingOverride=0.995
    linearSpringConstant=100
    simulationSpace=RootRelative
    alphaBoolBlend {blendOption=Linear}
    alphaScaleBiasClamp {interpSpeedDecreasing=3, interpSpeedIncreasing=5}
    boundBone {boneName=dyn_shard_skirt_e_base_l}
    componentAppliedLinearAccClamp=<6000 6000 6000>
    componentLinearAccScale=<0.05 0.05 0.05>
    componentLinearVelScale=<0.8 0.8 0.6>
    gravityOverride=<0.25 0.25 0.25>
    physicsBodyDefinitions [{
      sphereCollisionRadius=0
      boundBone {boneName=dyn_shard_skirt_e_base_l}
      boxExtents=<50 50 50>
      constraintSetup {angularLimitsMax=<0 0 0>, angularLimitsMin=<0 0 0>, angularTarget=<0 0 0>, linearAxesMax=<0.5 0.5 0>, linearAxesMin=<-0.5 -0.5 -0.5>}
      localJointOffset=<-3 0 0>
     }]
    relativeSpaceBone {boneName=root}
   }
   n60 = Anim Dynamics - Bone: dyn_shard_skirt_a_base_r(ComponentPose=@n59, Alpha=0.75, ExternalForce="0,0,0") {
    angularDampingOverride=0.999
    angularSpringConstant=150
    bAngularSpring=true
    bGravityOverrideInSimSpace=true
    bLinearSpring=true
    bOverrideAngularDamping=true
    bOverrideLinearDamping=true
    bUseGravityOverride=true
    gravityScale=0
    linearDampingOverride=0.995
    linearSpringConstant=100
    simulationSpace=RootRelative
    alphaBoolBlend {blendOption=Linear}
    alphaScaleBiasClamp {interpSpeedDecreasing=3, interpSpeedIncreasing=5}
    boundBone {boneName=dyn_shard_skirt_a_base_r}
    componentAppliedLinearAccClamp=<6000 6000 6000>
    componentLinearAccScale=<0.05 0.05 0.05>
    componentLinearVelScale=<0.8 0.8 0.6>
    gravityOverride=<0.25 0.25 0.25>
    physicsBodyDefinitions [{
      sphereCollisionRadius=0
      boundBone {boneName=dyn_shard_skirt_a_base_r}
      boxExtents=<50 50 50>
      constraintSetup {angularLimitsMax=<0 0 0>, angularLimitsMin=<0 0 0>, angularTarget=<0 0 0>, linearAxesMax=<0 0.5 0>, linearAxesMin=<-0.5 -0.5 -1>}
      localJointOffset=<-3 0 0>
     }]
    relativeSpaceBone {boneName=root}
   }
   n61 = Anim Dynamics - Bone: dyn_shard_skirt_b_r(ComponentPose=@n60, Alpha=1, ExternalForce="0,0,0") {
    angularDampingOverride=0.999
    angularSpringConstant=150
    bAngularSpring=true
    bGravityOverrideInSimSpace=true
    bLinearSpring=true
    bOverrideAngularDamping=true
    bOverrideLinearDamping=true
    bUseGravityOverride=true
    gravityScale=0
    linearDampingOverride=0.995
    linearSpringConstant=100
    simulationSpace=RootRelative
    alphaBoolBlend {blendOption=Linear}
    alphaScaleBiasClamp {interpSpeedDecreasing=3, interpSpeedIncreasing=5}
    boundBone {boneName=dyn_shard_skirt_b_r}
    componentAppliedLinearAccClamp=<6000 6000 6000>
    componentLinearAccScale=<0.05 0.05 0.05>
    componentLinearVelScale=<0.8 0.8 0.6>
    gravityOverride=<0.25 0.25 0.25>
    physicsBodyDefinitions [{
      sphereCollisionRadius=0
      boundBone {boneName=dyn_shard_skirt_b_r}
      boxExtents=<50 50 50>
      constraintSetup {angularLimitsMax=<0 0 0>, angularLimitsMin=<0 0 0>, angularTarget=<0 0 0>, linearAxesMax=<0 0.5 0>, linearAxesMin=<-2 -0.5 -1.5>}
      localJointOffset=<-3 0 0>
     }]
    relativeSpaceBone {boneName=root}
   }
   n62 = Anim Dynamics - Bone: dyn_shard_skirt_c_base_r(ComponentPose=@n61, Alpha=0.5, ExternalForce="0,0,0") {
    angularDampingOverride=0.999
    angularSpringConstant=150
    bAngularSpring=true
    bGravityOverrideInSimSpace=true
    bLinearSpring=true
    bOverrideAngularDamping=true
    bOverrideLinearDamping=true
    bUseGravityOverride=true
    gravityScale=0
    linearDampingOverride=0.995
    linearSpringConstant=100
    simulationSpace=RootRelative
    alphaBoolBlend {blendOption=Linear}
    alphaScaleBiasClamp {interpSpeedDecreasing=3, interpSpeedIncreasing=5}
    boundBone {boneName=dyn_shard_skirt_c_base_r}
    componentAppliedLinearAccClamp=<6000 6000 6000>
    componentLinearAccScale=<0.05 0.05 0.05>
    componentLinearVelScale=<0.8 0.8 0.6>
    gravityOverride=<0.25 0.25 0.25>
    physicsBodyDefinitions [{
      sphereCollisionRadius=0
      boundBone {boneName=dyn_shard_skirt_c_base_r}
      boxExtents=<50 50 50>
      constraintSetup {angularLimitsMax=<0 0 0>, angularLimitsMin=<0 0 0>, angularTarget=<0 0 0>, linearAxesMax=<0 0.5 0>, linearAxesMin=<-0.5 -0.5 -1>}
      localJointOffset=<-3 0 0>
     }]
    relativeSpaceBone {boneName=root}
   }
   n63 = Anim Dynamics - Bone: dyn_shard_skirt_d_base_r(ComponentPose=@n62, Alpha=0.4, ExternalForce="0,0,0") {
    angularDampingOverride=0.999
    angularSpringConstant=150
    bAngularSpring=true
    bGravityOverrideInSimSpace=true
    bLinearSpring=true
    bOverrideAngularDamping=true
    bOverrideLinearDamping=true
    bUseGravityOverride=true
    gravityScale=0
    linearDampingOverride=0.995
    linearSpringConstant=100
    simulationSpace=RootRelative
    alphaBoolBlend {blendOption=Linear}
    alphaScaleBiasClamp {interpSpeedDecreasing=3, interpSpeedIncreasing=5}
    boundBone {boneName=dyn_shard_skirt_d_base_r}
    componentAppliedLinearAccClamp=<6000 6000 6000>
    componentLinearAccScale=<0.05 0.05 0.05>
    componentLinearVelScale=<0.8 0.8 0.6>
    gravityOverride=<0.25 0.25 0.25>
    physicsBodyDefinitions [{
      sphereCollisionRadius=0
      boundBone {boneName=dyn_shard_skirt_d_base_r}
      boxExtents=<50 50 50>
      constraintSetup {angularLimitsMax=<0 0 0>, angularLimitsMin=<0 0 0>, angularTarget=<0 0 0>, linearAxesMax=<0 0.5 0>, linearAxesMin=<-1 -0.5 -1>}
      localJointOffset=<-3 0 0>
     }]
    relativeSpaceBone {boneName=root}
   }
   n64 = Anim Dynamics - Bone: dyn_shard_skirt_e_base_r(ComponentPose=@n63, Alpha=0.4, ExternalForce="0,0,0") {
    angularDampingOverride=0.999
    angularSpringConstant=150
    bAngularSpring=true
    bGravityOverrideInSimSpace=true
    bLinearSpring=true
    bOverrideAngularDamping=true
    bOverrideLinearDamping=true
    bUseGravityOverride=true
    gravityScale=0
    linearDampingOverride=0.995
    linearSpringConstant=100
    simulationSpace=RootRelative
    alphaBoolBlend {blendOption=Linear}
    alphaScaleBiasClamp {interpSpeedDecreasing=3, interpSpeedIncreasing=5}
    boundBone {boneName=dyn_shard_skirt_e_base_r}
    componentAppliedLinearAccClamp=<6000 6000 6000>
    componentLinearAccScale=<0.05 0.05 0.05>
    componentLinearVelScale=<0.8 0.8 0.6>
    gravityOverride=<0.25 0.25 0.25>
    physicsBodyDefinitions [{
      sphereCollisionRadius=0
      boundBone {boneName=dyn_shard_skirt_e_base_r}
      boxExtents=<50 50 50>
      constraintSetup {angularLimitsMax=<0 0 0>, angularLimitsMin=<0 0 0>, angularTarget=<0 0 0>, linearAxesMax=<1 0.5 0>, linearAxesMin=<-1 -0.5 -1>}
      localJointOffset=<-3 0 0>
     }]
    relativeSpaceBone {boneName=root}
   }
   n65 = Anim Dynamics - Bone: dyn_shard_skirt_f_base_r(ComponentPose=@n64, Alpha=1, ExternalForce="0,0,0") {
    angularDampingOverride=0.999
    angularSpringConstant=150
    bAngularSpring=true
    bGravityOverrideInSimSpace=true
    bLinearSpring=true
    bOverrideAngularDamping=true
    bOverrideLinearDamping=true
    bUseGravityOverride=true
    gravityScale=0
    linearDampingOverride=0.995
    linearSpringConstant=100
    simulationSpace=RootRelative
    alphaBoolBlend {blendOption=Linear}
    alphaScaleBiasClamp {interpSpeedDecreasing=3, interpSpeedIncreasing=5}
    boundBone {boneName=dyn_shard_skirt_f_base_r}
    componentAppliedLinearAccClamp=<6000 6000 6000>
    componentLinearAccScale=<0.05 0.05 0.05>
    componentLinearVelScale=<0.8 0.8 0.6>
    gravityOverride=<0.25 0.25 0.25>
    physicsBodyDefinitions [{
      sphereCollisionRadius=0
      boundBone {boneName=dyn_shard_skirt_f_base_r}
      boxExtents=<50 50 50>
      constraintSetup {angularLimitsMax=<0 0 0>, angularLimitsMin=<0 0 0>, angularTarget=<0 0 0>, linearAxesMax=<1 0.5 0.5>, linearAxesMin=<-2 -0.5 -2>}
      localJointOffset=<-3 0 0>
     }]
    relativeSpaceBone {boneName=root}
   }
   n66 = Anim Dynamics - Bone: dyn_shard_skirt_g_base_r(ComponentPose=@n65, Alpha=0.3, ExternalForce="0,0,0") {
    angularDampingOverride=0.999
    angularSpringConstant=150
    bAngularSpring=true
    bGravityOverrideInSimSpace=true
    bLinearSpring=true
    bOverrideAngularDamping=true
    bOverrideLinearDamping=true
    bUseGravityOverride=true
    gravityScale=0
    linearDampingOverride=0.995
    linearSpringConstant=100
    simulationSpace=RootRelative
    alphaBoolBlend {blendOption=Linear}
    alphaScaleBiasClamp {interpSpeedDecreasing=3, interpSpeedIncreasing=5}
    boundBone {boneName=dyn_shard_skirt_g_base_r}
    componentAppliedLinearAccClamp=<6000 6000 6000>
    componentLinearAccScale=<0.05 0.05 0.05>
    componentLinearVelScale=<0.8 0.8 0.6>
    gravityOverride=<0.25 0.25 0.25>
    physicsBodyDefinitions [{
      sphereCollisionRadius=0
      boundBone {boneName=dyn_shard_skirt_g_base_r}
      boxExtents=<50 50 50>
      constraintSetup {angularLimitsMax=<0 0 0>, angularLimitsMin=<0 0 0>, angularTarget=<0 0 0>, linearAxesMax=<0.5 0.5 0.5>, linearAxesMin=<-0.5 -0.5 -0.5>}
      localJointOffset=<-3 0 0>
     }]
    relativeSpaceBone {boneName=root}
   }
   n67 = Anim Dynamics - Bone: dyn_shard_skirt_h_base_r(ComponentPose=@n66, Alpha=1, ExternalForce="0,0,0") {
    angularDampingOverride=0.999
    angularSpringConstant=150
    bAngularSpring=true
    bGravityOverrideInSimSpace=true
    bLinearSpring=true
    bOverrideAngularDamping=true
    bOverrideLinearDamping=true
    bUseGravityOverride=true
    gravityScale=0
    linearDampingOverride=0.995
    linearSpringConstant=100
    simulationSpace=RootRelative
    alphaBoolBlend {blendOption=Linear}
    alphaScaleBiasClamp {interpSpeedDecreasing=3, interpSpeedIncreasing=5}
    boundBone {boneName=dyn_shard_skirt_h_base_r}
    componentAppliedLinearAccClamp=<6000 6000 6000>
    componentLinearAccScale=<0.05 0.05 0.05>
    componentLinearVelScale=<0.8 0.8 0.6>
    gravityOverride=<0.25 0.25 0.25>
    physicsBodyDefinitions [{
      sphereCollisionRadius=0
      boundBone {boneName=dyn_shard_skirt_h_base_r}
      boxExtents=<50 50 50>
      constraintSetup {angularLimitsMax=<0 0 0>, angularLimitsMin=<0 0 0>, angularTarget=<0 0 0>, linearAxesMax=<1 0.5 0.5>, linearAxesMin=<-1 -0.5 -0.5>}
      localJointOffset=<-3 0 0>
     }]
    relativeSpaceBone {boneName=root}
   }
   n68 = Anim Dynamics - Bone: dyn_shard_skirt_i_base_r(ComponentPose=@n67, Alpha=0.3, ExternalForce="0,0,0") {
    angularDampingOverride=0.999
    angularSpringConstant=150
    bAngularSpring=true
    bGravityOverrideInSimSpace=true
    bLinearSpring=true
    bOverrideAngularDamping=true
    bOverrideLinearDamping=true
    bUseGravityOverride=true
    gravityScale=0
    linearDampingOverride=0.995
    linearSpringConstant=100
    simulationSpace=RootRelative
    alphaBoolBlend {blendOption=Linear}
    alphaScaleBiasClamp {interpSpeedDecreasing=3, interpSpeedIncreasing=5}
    boundBone {boneName=dyn_shard_skirt_i_base_r}
    componentAppliedLinearAccClamp=<6000 6000 6000>
    componentLinearAccScale=<0.05 0.05 0.05>
    componentLinearVelScale=<0.8 0.8 0.6>
    gravityOverride=<0.25 0.25 0.25>
    physicsBodyDefinitions [{
      sphereCollisionRadius=0
      boundBone {boneName=dyn_shard_skirt_i_base_r}
      boxExtents=<50 50 50>
      constraintSetup {angularLimitsMax=<0 0 0>, angularLimitsMin=<0 0 0>, angularTarget=<0 0 0>, linearAxesMax=<0.5 0.5 0.5>, linearAxesMin=<-0.5 -0.5 -0.5>}
      localJointOffset=<-3 0 0>
     }]
    relativeSpaceBone {boneName=root}
   }
   n69 = Anim Dynamics - Bone: dyn_shard_skirt_j_base_r(ComponentPose=@n68, Alpha=0.3, ExternalForce="0,0,0") {
    angularDampingOverride=0.999
    angularSpringConstant=150
    bAngularSpring=true
    bGravityOverrideInSimSpace=true
    bLinearSpring=true
    bOverrideAngularDamping=true
    bOverrideLinearDamping=true
    bUseGravityOverride=true
    gravityScale=0
    linearDampingOverride=0.995
    linearSpringConstant=100
    simulationSpace=RootRelative
    alphaBoolBlend {blendOption=Linear}
    alphaScaleBiasClamp {interpSpeedDecreasing=3, interpSpeedIncreasing=5}
    boundBone {boneName=dyn_shard_skirt_j_base_r}
    componentAppliedLinearAccClamp=<6000 6000 6000>
    componentLinearAccScale=<0.05 0.05 0.05>
    componentLinearVelScale=<0.8 0.8 0.6>
    gravityOverride=<0.25 0.25 0.25>
    physicsBodyDefinitions [{
      sphereCollisionRadius=0
      boundBone {boneName=dyn_shard_skirt_j_base_r}
      boxExtents=<50 50 50>
      constraintSetup {angularLimitsMax=<0 0 0>, angularLimitsMin=<0 0 0>, angularTarget=<0 0 0>, linearAxesMax=<1 0.5 0.5>, linearAxesMin=<-1 -0.5 -0.5>}
      localJointOffset=<-3 0 0>
     }]
    relativeSpaceBone {boneName=root}
   }
   n70 = Anim Dynamics - Bone: dyn_shard_shoulder_a_base_l(ComponentPose=@n69, Alpha=0.25, ExternalForce="0,0,0") {
    angularDampingOverride=0.999
    angularSpringConstant=150
    bAngularSpring=true
    bGravityOverrideInSimSpace=true
    bLinearSpring=true
    bOverrideAngularDamping=true
    bOverrideLinearDamping=true
    bUseGravityOverride=true
    gravityScale=0
    linearDampingOverride=0.995
    linearSpringConstant=100
    simulationSpace=RootRelative
    alphaBoolBlend {blendOption=Linear}
    alphaScaleBiasClamp {interpSpeedDecreasing=3, interpSpeedIncreasing=5}
    boundBone {boneName=dyn_shard_shoulder_a_base_l}
    componentAppliedLinearAccClamp=<6000 6000 6000>
    componentLinearAccScale=<0.05 0.05 0.05>
    componentLinearVelScale=<0.8 0.8 0.6>
    gravityOverride=<0.25 0.25 0.25>
    physicsBodyDefinitions [{
      sphereCollisionRadius=0
      boundBone {boneName=dyn_shard_shoulder_a_base_l}
      boxExtents=<50 50 50>
      constraintSetup {angularLimitsMax=<0 0 0>, angularLimitsMin=<0 0 0>, angularTarget=<0 0 0>, linearAxesMax=<0.5 0.5 0.5>, linearAxesMin=<-0.5 -0.5 -0.5>}
      localJointOffset=<-3 0 0>
     }]
    relativeSpaceBone {boneName=root}
   }
   n71 = Anim Dynamics - Bone: dyn_shard_shoulder_b_base_l(ComponentPose=@n70, Alpha=0.3, ExternalForce="0,0,0") {
    angularDampingOverride=0.999
    angularSpringConstant=150
    bAngularSpring=true
    bGravityOverrideInSimSpace=true
    bLinearSpring=true
    bOverrideAngularDamping=true
    bOverrideLinearDamping=true
    bUseGravityOverride=true
    gravityScale=0
    linearDampingOverride=0.995
    linearSpringConstant=100
    simulationSpace=RootRelative
    alphaBoolBlend {blendOption=Linear}
    alphaScaleBiasClamp {interpSpeedDecreasing=3, interpSpeedIncreasing=5}
    boundBone {boneName=dyn_shard_shoulder_b_base_l}
    componentAppliedLinearAccClamp=<6000 6000 6000>
    componentLinearAccScale=<0.05 0.05 0.05>
    componentLinearVelScale=<0.8 0.8 0.6>
    gravityOverride=<0.25 0.25 0.25>
    physicsBodyDefinitions [{
      sphereCollisionRadius=0
      boundBone {boneName=dyn_shard_shoulder_b_base_l}
      boxExtents=<50 50 50>
      constraintSetup {angularLimitsMax=<0 0 0>, angularLimitsMin=<0 0 0>, angularTarget=<0 0 0>, linearAxesMax=<0.5 0.5 0.5>, linearAxesMin=<-0.5 -0.5 -0.5>}
      localJointOffset=<-3 0 0>
     }]
    relativeSpaceBone {boneName=root}
   }
   n72 = Anim Dynamics - Bone: dyn_shard_shoulder_c_base_l(ComponentPose=@n71, Alpha=0.8, ExternalForce="0,0,0") {
    angularDampingOverride=0.999
    angularSpringConstant=150
    bAngularSpring=true
    bGravityOverrideInSimSpace=true
    bLinearSpring=true
    bOverrideAngularDamping=true
    bOverrideLinearDamping=true
    bUseGravityOverride=true
    gravityScale=0
    linearDampingOverride=0.995
    linearSpringConstant=100
    simulationSpace=RootRelative
    alphaBoolBlend {blendOption=Linear}
    alphaScaleBiasClamp {interpSpeedDecreasing=3, interpSpeedIncreasing=5}
    boundBone {boneName=dyn_shard_shoulder_c_base_l}
    componentAppliedLinearAccClamp=<6000 6000 6000>
    componentLinearAccScale=<0.05 0.05 0.05>
    componentLinearVelScale=<0.8 0.8 0.6>
    gravityOverride=<0.25 0.25 0.25>
    physicsBodyDefinitions [{
      sphereCollisionRadius=0
      boundBone {boneName=dyn_shard_shoulder_c_base_l}
      boxExtents=<50 50 50>
      constraintSetup {angularLimitsMax=<0 0 0>, angularLimitsMin=<0 0 0>, angularTarget=<0 0 0>, linearAxesMax=<0.5 0.5 0.5>, linearAxesMin=<-0.5 -0.5 -0.5>}
      localJointOffset=<-3 0 0>
     }]
    relativeSpaceBone {boneName=root}
   }
   n73 = Anim Dynamics - Bone: dyn_shard_shoulder_d_base_l(ComponentPose=@n72, Alpha=0.9, ExternalForce="0,0,0") {
    angularDampingOverride=0.999
    angularSpringConstant=150
    bAngularSpring=true
    bGravityOverrideInSimSpace=true
    bLinearSpring=true
    bOverrideAngularDamping=true
    bOverrideLinearDamping=true
    bUseGravityOverride=true
    gravityScale=0
    linearDampingOverride=0.995
    linearSpringConstant=100
    simulationSpace=RootRelative
    alphaBoolBlend {blendOption=Linear}
    alphaScaleBiasClamp {interpSpeedDecreasing=3, interpSpeedIncreasing=5}
    boundBone {boneName=dyn_shard_shoulder_d_base_l}
    componentAppliedLinearAccClamp=<6000 6000 6000>
    componentLinearAccScale=<0.05 0.05 0.05>
    componentLinearVelScale=<0.8 0.8 0.6>
    gravityOverride=<0.25 0.25 0.25>
    physicsBodyDefinitions [{
      sphereCollisionRadius=0
      boundBone {boneName=dyn_shard_shoulder_d_base_l}
      boxExtents=<50 50 50>
      constraintSetup {angularLimitsMax=<0 0 0>, angularLimitsMin=<0 0 0>, angularTarget=<0 0 0>, linearAxesMax=<0.5 0.5 0.5>, linearAxesMin=<-0.5 -0.5 -0.5>}
      localJointOffset=<-3 0 0>
     }]
    relativeSpaceBone {boneName=root}
   }
   n74 = Anim Dynamics - Bone: dyn_shard_shoulder_e_base_l(ComponentPose=@n73, Alpha=0.7, ExternalForce="0,0,0") {
    angularDampingOverride=0.999
    angularSpringConstant=150
    bAngularSpring=true
    bGravityOverrideInSimSpace=true
    bLinearSpring=true
    bOverrideAngularDamping=true
    bOverrideLinearDamping=true
    bUseGravityOverride=true
    gravityScale=0
    linearDampingOverride=0.995
    linearSpringConstant=100
    simulationSpace=RootRelative
    alphaBoolBlend {blendOption=Linear}
    alphaScaleBiasClamp {interpSpeedDecreasing=3, interpSpeedIncreasing=5}
    boundBone {boneName=dyn_shard_shoulder_e_base_l}
    componentAppliedLinearAccClamp=<6000 6000 6000>
    componentLinearAccScale=<0.05 0.05 0.05>
    componentLinearVelScale=<0.8 0.8 0.6>
    gravityOverride=<0.25 0.25 0.25>
    physicsBodyDefinitions [{
      sphereCollisionRadius=0
      boundBone {boneName=dyn_shard_shoulder_e_base_l}
      boxExtents=<50 50 50>
      constraintSetup {angularLimitsMax=<0 0 0>, angularLimitsMin=<0 0 0>, angularTarget=<0 0 0>, linearAxesMax=<0.5 0.5 0.5>, linearAxesMin=<-0.5 -0.5 -0.5>}
      localJointOffset=<-3 0 0>
     }]
    relativeSpaceBone {boneName=root}
   }
   n75 = Anim Dynamics - Bone: dyn_shard_shoulder_f_base_l(ComponentPose=@n74, Alpha=1, ExternalForce="0,0,0") {
    angularDampingOverride=0.999
    angularSpringConstant=150
    bAngularSpring=true
    bGravityOverrideInSimSpace=true
    bLinearSpring=true
    bOverrideAngularDamping=true
    bOverrideLinearDamping=true
    bUseGravityOverride=true
    gravityScale=0
    linearDampingOverride=0.995
    linearSpringConstant=100
    simulationSpace=RootRelative
    alphaBoolBlend {blendOption=Linear}
    alphaScaleBiasClamp {interpSpeedDecreasing=3, interpSpeedIncreasing=5}
    boundBone {boneName=dyn_shard_shoulder_f_base_l}
    componentAppliedLinearAccClamp=<6000 6000 6000>
    componentLinearAccScale=<0.05 0.05 0.05>
    componentLinearVelScale=<0.8 0.8 0.6>
    gravityOverride=<0.25 0.25 0.25>
    physicsBodyDefinitions [{
      sphereCollisionRadius=0
      boundBone {boneName=dyn_shard_shoulder_f_base_l}
      boxExtents=<50 50 50>
      constraintSetup {angularLimitsMax=<0 0 0>, angularLimitsMin=<0 0 0>, angularTarget=<0 0 0>, linearAxesMax=<0.5 0.5 0.5>, linearAxesMin=<-0.5 -0.5 -0.5>}
      localJointOffset=<-3 0 0>
     }]
    relativeSpaceBone {boneName=root}
   }
   n76 = Anim Dynamics - Bone: dyn_shard_shoulder_g_base_l(ComponentPose=@n75, Alpha=0.4, ExternalForce="0,0,0") {
    angularDampingOverride=0.999
    angularSpringConstant=150
    bAngularSpring=true
    bGravityOverrideInSimSpace=true
    bLinearSpring=true
    bOverrideAngularDamping=true
    bOverrideLinearDamping=true
    bUseGravityOverride=true
    gravityScale=0
    linearDampingOverride=0.995
    linearSpringConstant=100
    simulationSpace=RootRelative
    alphaBoolBlend {blendOption=Linear}
    alphaScaleBiasClamp {interpSpeedDecreasing=3, interpSpeedIncreasing=5}
    boundBone {boneName=dyn_shard_shoulder_g_base_l}
    componentAppliedLinearAccClamp=<6000 6000 6000>
    componentLinearAccScale=<0.05 0.05 0.05>
    componentLinearVelScale=<0.8 0.8 0.6>
    gravityOverride=<0.25 0.25 0.25>
    physicsBodyDefinitions [{
      sphereCollisionRadius=0
      boundBone {boneName=dyn_shard_shoulder_g_base_l}
      boxExtents=<50 50 50>
      constraintSetup {angularLimitsMax=<0 0 0>, angularLimitsMin=<0 0 0>, angularTarget=<0 0 0>, linearAxesMax=<0.5 0.5 0.5>, linearAxesMin=<-0.5 -0.5 -0.5>}
      localJointOffset=<-3 0 0>
     }]
    relativeSpaceBone {boneName=root}
   }
   n77 = Anim Dynamics - Bone: dyn_shard_shoulder_h_base_l(ComponentPose=@n76, Alpha=0.5, ExternalForce="0,0,0") {
    angularDampingOverride=0.999
    angularSpringConstant=150
    bAngularSpring=true
    bGravityOverrideInSimSpace=true
    bLinearSpring=true
    bOverrideAngularDamping=true
    bOverrideLinearDamping=true
    bUseGravityOverride=true
    gravityScale=0
    linearDampingOverride=0.995
    linearSpringConstant=100
    simulationSpace=RootRelative
    alphaBoolBlend {blendOption=Linear}
    alphaScaleBiasClamp {interpSpeedDecreasing=3, interpSpeedIncreasing=5}
    boundBone {boneName=dyn_shard_shoulder_h_base_l}
    componentAppliedLinearAccClamp=<6000 6000 6000>
    componentLinearAccScale=<0.05 0.05 0.05>
    componentLinearVelScale=<0.8 0.8 0.6>
    gravityOverride=<0.25 0.25 0.25>
    physicsBodyDefinitions [{
      sphereCollisionRadius=0
      boundBone {boneName=dyn_shard_shoulder_h_base_l}
      boxExtents=<50 50 50>
      constraintSetup {angularLimitsMax=<0 0 0>, angularLimitsMin=<0 0 0>, angularTarget=<0 0 0>, linearAxesMax=<0.5 0.5 0.5>, linearAxesMin=<-0.5 -0.5 -0.5>}
      localJointOffset=<-3 0 0>
     }]
    relativeSpaceBone {boneName=root}
   }
   n78 = Anim Dynamics - Bone: dyn_shard_elbow_a_base_r(ComponentPose=@n77, Alpha=0.7, ExternalForce="0,0,0") {
    angularDampingOverride=0.999
    angularSpringConstant=150
    bAngularSpring=true
    bGravityOverrideInSimSpace=true
    bLinearSpring=true
    bOverrideAngularDamping=true
    bOverrideLinearDamping=true
    bUseGravityOverride=true
    gravityScale=0
    linearDampingOverride=0.995
    linearSpringConstant=100
    simulationSpace=RootRelative
    alphaBoolBlend {blendOption=Linear}
    alphaScaleBiasClamp {interpSpeedDecreasing=3, interpSpeedIncreasing=5}
    boundBone {boneName=dyn_shard_elbow_a_base_r}
    componentAppliedLinearAccClamp=<6000 6000 6000>
    componentLinearAccScale=<0.05 0.05 0.05>
    componentLinearVelScale=<0.8 0.8 0.6>
    gravityOverride=<0.25 0.25 0.25>
    physicsBodyDefinitions [{
      sphereCollisionRadius=0
      boundBone {boneName=dyn_shard_elbow_a_base_r}
      boxExtents=<50 50 50>
      constraintSetup {angularLimitsMax=<0 0 0>, angularLimitsMin=<0 0 0>, angularTarget=<0 0 0>, linearAxesMax=<0.5 0.5 0.5>, linearAxesMin=<-0.5 -0.5 -0.5>}
      localJointOffset=<-3 0 0>
     }]
    relativeSpaceBone {boneName=root}
   }
   n79 = Anim Dynamics - Bone: dyn_shard_elbow_b_base_r(ComponentPose=@n78, Alpha=0.5, ExternalForce="0,0,0") {
    angularDampingOverride=0.999
    angularSpringConstant=150
    bAngularSpring=true
    bGravityOverrideInSimSpace=true
    bLinearSpring=true
    bOverrideAngularDamping=true
    bOverrideLinearDamping=true
    bUseGravityOverride=true
    gravityScale=0
    linearDampingOverride=0.995
    linearSpringConstant=100
    simulationSpace=RootRelative
    alphaBoolBlend {blendOption=Linear}
    alphaScaleBiasClamp {interpSpeedDecreasing=3, interpSpeedIncreasing=5}
    boundBone {boneName=dyn_shard_elbow_b_base_r}
    componentAppliedLinearAccClamp=<6000 6000 6000>
    componentLinearAccScale=<0.05 0.05 0.05>
    componentLinearVelScale=<0.8 0.8 0.6>
    gravityOverride=<0.25 0.25 0.25>
    physicsBodyDefinitions [{
      sphereCollisionRadius=0
      boundBone {boneName=dyn_shard_elbow_b_base_r}
      boxExtents=<50 50 50>
      constraintSetup {angularLimitsMax=<0 0 0>, angularLimitsMin=<0 0 0>, angularTarget=<0 0 0>, linearAxesMax=<0.5 0.5 0.5>, linearAxesMin=<-0.5 -0.5 -0.5>}
      localJointOffset=<-3 0 0>
     }]
    relativeSpaceBone {boneName=root}
   }
   n80 = Anim Dynamics - Bone: dyn_shard_knee_a_base_l(ComponentPose=@n79, Alpha=1, ExternalForce="0,0,0") {
    angularDampingOverride=0.999
    angularSpringConstant=150
    bAngularSpring=true
    bGravityOverrideInSimSpace=true
    bLinearSpring=true
    bOverrideAngularDamping=true
    bOverrideLinearDamping=true
    bUseGravityOverride=true
    gravityScale=0
    linearDampingOverride=0.995
    linearSpringConstant=100
    simulationSpace=RootRelative
    alphaBoolBlend {blendOption=Linear}
    alphaScaleBiasClamp {interpSpeedDecreasing=3, interpSpeedIncreasing=5}
    boundBone {boneName=dyn_shard_knee_a_base_l}
    componentAppliedLinearAccClamp=<6000 6000 6000>
    componentLinearAccScale=<0.05 0.05 0.05>
    componentLinearVelScale=<0.8 0.8 0.6>
    gravityOverride=<0.25 0.25 0.25>
    physicsBodyDefinitions [{
      sphereCollisionRadius=0
      boundBone {boneName=dyn_shard_knee_a_base_l}
      boxExtents=<50 50 50>
      constraintSetup {angularLimitsMax=<0 0 0>, angularLimitsMin=<0 0 0>, angularTarget=<0 0 0>, linearAxesMax=<0.5 0.5 0.5>, linearAxesMin=<-0.5 -0.5 -0.5>}
      localJointOffset=<-3 0 0>
     }]
    relativeSpaceBone {boneName=root}
   }
   n81 = Anim Dynamics - Bone: dyn_shard_torso_a_base_l(ComponentPose=@n80, Alpha=1, ExternalForce="0,0,0") {
    angularDampingOverride=0.999
    angularSpringConstant=150
    bAngularSpring=true
    bGravityOverrideInSimSpace=true
    bLinearSpring=true
    bOverrideAngularDamping=true
    bOverrideLinearDamping=true
    bUseGravityOverride=true
    gravityScale=0
    linearDampingOverride=0.995
    linearSpringConstant=100
    simulationSpace=RootRelative
    alphaBoolBlend {blendOption=Linear}
    alphaScaleBiasClamp {interpSpeedDecreasing=3, interpSpeedIncreasing=5}
    boundBone {boneName=dyn_shard_torso_a_base_l}
    componentAppliedLinearAccClamp=<6000 6000 6000>
    componentLinearAccScale=<0.05 0.05 0.05>
    componentLinearVelScale=<0.8 0.8 0.6>
    gravityOverride=<0.25 0.25 0.25>
    physicsBodyDefinitions [{
      sphereCollisionRadius=0
      boundBone {boneName=dyn_shard_torso_a_base_l}
      boxExtents=<50 50 50>
      constraintSetup {angularLimitsMax=<0 0 0>, angularLimitsMin=<0 0 0>, angularTarget=<0 0 0>, linearAxesMax=<0.5 0.5 0.5>, linearAxesMin=<-0.5 -0.5 -0.5>}
      localJointOffset=<-3 0 0>
     }]
    relativeSpaceBone {boneName=root}
   }
   n82 = "Transform (Modify) Bone - Bone: dyn_shard_skirt_a_l"(ComponentPose=@n81, Alpha=0.3, Translation=Get floatA() -> (floatA:Vector), Rotation="0,0,0", Scale="1,1,1") {
    rotationMode=BMM_Additive
    rotationSpace=BCS_BoneSpace
    scaleMode=BMM_Additive
    scaleSpace=BCS_BoneSpace
    translationMode=BMM_Additive
    translationSpace=BCS_BoneSpace
    alphaBoolBlend {blendOption=Linear}
    boneToModify {boneName=dyn_shard_skirt_a_l}
   }
   n83 = "Transform (Modify) Bone - Bone: dyn_shard_skirt_b_l"(ComponentPose=@n82, Alpha=0.7, Translation=Get floatB() -> (floatB:Vector), Rotation="0,0,0", Scale="1,1,1") {
    rotationMode=BMM_Additive
    rotationSpace=BCS_BoneSpace
    scaleMode=BMM_Additive
    scaleSpace=BCS_BoneSpace
    translationMode=BMM_Additive
    translationSpace=BCS_BoneSpace
    alphaBoolBlend {blendOption=Linear}
    boneToModify {boneName=dyn_shard_skirt_b_l}
   }
   n84 = "Transform (Modify) Bone - Bone: dyn_shard_skirt_c_l"(ComponentPose=@n83, Alpha=1, Translation=Get floatA(), Rotation="0,0,0", Scale="1,1,1") {
    rotationMode=BMM_Additive
    rotationSpace=BCS_BoneSpace
    scaleMode=BMM_Additive
    scaleSpace=BCS_BoneSpace
    translationMode=BMM_Additive
    translationSpace=BCS_BoneSpace
    alphaBoolBlend {blendOption=Linear}
    boneToModify {boneName=dyn_shard_skirt_c_l}
   }
   n85 = "Transform (Modify) Bone - Bone: dyn_shard_skirt_d_l"(ComponentPose=@n84, Alpha=0.7, Translation=Get floatB(), Rotation="0,0,0", Scale="1,1,1") {
    rotationMode=BMM_Additive
    rotationSpace=BCS_BoneSpace
    scaleMode=BMM_Additive
    scaleSpace=BCS_BoneSpace
    translationMode=BMM_Additive
    translationSpace=BCS_BoneSpace
    alphaBoolBlend {blendOption=Linear}
    boneToModify {boneName=dyn_shard_skirt_d_l}
   }
   n86 = "Transform (Modify) Bone - Bone: dyn_shard_skirt_e_l"(ComponentPose=@n85, Alpha=0.5, Translation=Get floatA(), Rotation="0,0,0", Scale="1,1,1") {
    rotationMode=BMM_Additive
    rotationSpace=BCS_BoneSpace
    scaleMode=BMM_Additive
    scaleSpace=BCS_BoneSpace
    translationMode=BMM_Additive
    translationSpace=BCS_BoneSpace
    alphaBoolBlend {blendOption=Linear}
    boneToModify {boneName=dyn_shard_skirt_e_l}
   }
   n87 = "Transform (Modify) Bone - Bone: dyn_shard_skirt_a_r"(ComponentPose=@n86, Alpha=0.75, Translation=Get floatA(), Rotation="0,0,0", Scale="1,1,1") {
    rotationMode=BMM_Additive
    rotationSpace=BCS_BoneSpace
    scaleMode=BMM_Additive
    scaleSpace=BCS_BoneSpace
    translationMode=BMM_Additive
    translationSpace=BCS_BoneSpace
    alphaBoolBlend {blendOption=Linear}
    boneToModify {boneName=dyn_shard_skirt_a_r}
   }
   n88 = "Transform (Modify) Bone - Bone: dyn_shard_skirt_b_r"(ComponentPose=@n87, Alpha=1, Translation=Get floatB(), Rotation="0,0,0", Scale="1,1,1") {
    rotationMode=BMM_Additive
    rotationSpace=BCS_BoneSpace
    scaleMode=BMM_Additive
    scaleSpace=BCS_BoneSpace
    translationMode=BMM_Additive
    translationSpace=BCS_BoneSpace
    alphaBoolBlend {blendOption=Linear}
    boneToModify {boneName=dyn_shard_skirt_b_r}
   }
   n89 = "Transform (Modify) Bone - Bone: dyn_shard_skirt_c_r"(ComponentPose=@n88, Alpha=1, Translation=Get floatA(), Rotation="0,0,0", Scale="1,1,1") {
    rotationMode=BMM_Additive
    rotationSpace=BCS_BoneSpace
    scaleMode=BMM_Additive
    scaleSpace=BCS_BoneSpace
    translationMode=BMM_Additive
    translationSpace=BCS_BoneSpace
    alphaBoolBlend {blendOption=Linear}
    boneToModify {boneName=dyn_shard_skirt_c_r}
   }
   n90 = "Transform (Modify) Bone - Bone: dyn_shard_skirt_d_r"(ComponentPose=@n89, Alpha=0.3, Translation=Get floatB(), Rotation="0,0,0", Scale="1,1,1") {
    rotationMode=BMM_Additive
    rotationSpace=BCS_BoneSpace
    scaleMode=BMM_Additive
    scaleSpace=BCS_BoneSpace
    translationMode=BMM_Additive
    translationSpace=BCS_BoneSpace
    alphaBoolBlend {blendOption=Linear}
    boneToModify {boneName=dyn_shard_skirt_d_r}
   }
   n91 = "Transform (Modify) Bone - Bone: dyn_shard_skirt_e_r"(ComponentPose=@n90, Alpha=1, Translation=Get floatA(), Rotation="0,0,0", Scale="1,1,1") {
    rotationMode=BMM_Additive
    rotationSpace=BCS_BoneSpace
    scaleMode=BMM_Additive
    scaleSpace=BCS_BoneSpace
    translationMode=BMM_Additive
    translationSpace=BCS_BoneSpace
    alphaBoolBlend {blendOption=Linear}
    boneToModify {boneName=dyn_shard_skirt_e_r}
   }
   n92 = "Transform (Modify) Bone - Bone: dyn_shard_skirt_f_r"(ComponentPose=@n91, Alpha=1, Translation=Get floatB(), Rotation="0,0,0", Scale="1,1,1") {
    rotationMode=BMM_Additive
    rotationSpace=BCS_BoneSpace
    scaleMode=BMM_Additive
    scaleSpace=BCS_BoneSpace
    translationMode=BMM_Additive
    translationSpace=BCS_BoneSpace
    alphaBoolBlend {blendOption=Linear}
    boneToModify {boneName=dyn_shard_skirt_f_r}
   }
   n93 = "Transform (Modify) Bone - Bone: dyn_shard_skirt_g_r"(ComponentPose=@n92, Alpha=0.3, Translation=Get floatA(), Rotation="0,0,0", Scale="1,1,1") {
    rotationMode=BMM_Additive
    rotationSpace=BCS_BoneSpace
    scaleMode=BMM_Additive
    scaleSpace=BCS_BoneSpace
    translationMode=BMM_Additive
    translationSpace=BCS_BoneSpace
    alphaBoolBlend {blendOption=Linear}
    boneToModify {boneName=dyn_shard_skirt_g_r}
   }
   n94 = "Transform (Modify) Bone - Bone: dyn_shard_skirt_h_r"(ComponentPose=@n93, Alpha=1, Translation=Get floatB(), Rotation="0,0,0", Scale="1,1,1") {
    rotationMode=BMM_Additive
    rotationSpace=BCS_BoneSpace
    scaleMode=BMM_Additive
    scaleSpace=BCS_BoneSpace
    translationMode=BMM_Additive
    translationSpace=BCS_BoneSpace
    alphaBoolBlend {blendOption=Linear}
    boneToModify {boneName=dyn_shard_skirt_h_r}
   }
   n95 = "Transform (Modify) Bone - Bone: dyn_shard_skirt_i_r"(ComponentPose=@n94, Alpha=0.5, Translation=Get floatA(), Rotation="0,0,0", Scale="1,1,1") {
    rotationMode=BMM_Additive
    rotationSpace=BCS_BoneSpace
    scaleMode=BMM_Additive
    scaleSpace=BCS_BoneSpace
    translationMode=BMM_Additive
    translationSpace=BCS_BoneSpace
    alphaBoolBlend {blendOption=Linear}
    boneToModify {boneName=dyn_shard_skirt_i_r}
   }
   n96 = "Transform (Modify) Bone - Bone: dyn_shard_skirt_j_r"(ComponentPose=@n95, Alpha=0.5, Translation=Get floatB(), Rotation="0,0,0", Scale="1,1,1") {
    rotationMode=BMM_Additive
    rotationSpace=BCS_BoneSpace
    scaleMode=BMM_Additive
    scaleSpace=BCS_BoneSpace
    translationMode=BMM_Additive
    translationSpace=BCS_BoneSpace
    alphaBoolBlend {blendOption=Linear}
    boneToModify {boneName=dyn_shard_skirt_j_r}
   }
   n97 = "Transform (Modify) Bone - Bone: dyn_shard_elbow_a_r"(ComponentPose=@n96, Alpha=1, Translation=Get floatB(), Rotation="0,0,0", Scale="1,1,1") {
    rotationMode=BMM_Additive
    rotationSpace=BCS_BoneSpace
    scaleMode=BMM_Additive
    scaleSpace=BCS_BoneSpace
    translationMode=BMM_Additive
    translationSpace=BCS_BoneSpace
    alphaBoolBlend {blendOption=Linear}
    boneToModify {boneName=dyn_shard_elbow_a_r}
   }
   n98 = "Transform (Modify) Bone - Bone: dyn_shard_elbow_b_r"(ComponentPose=@n97, Alpha=0.5, Translation=Get floatA(), Rotation="0,0,0", Scale="1,1,1") {
    rotationMode=BMM_Additive
    rotationSpace=BCS_BoneSpace
    scaleMode=BMM_Additive
    scaleSpace=BCS_BoneSpace
    translationMode=BMM_Additive
    translationSpace=BCS_BoneSpace
    alphaBoolBlend {blendOption=Linear}
    boneToModify {boneName=dyn_shard_elbow_b_r}
   }
   n99 = "Transform (Modify) Bone - Bone: dyn_shard_knee_a_l"(ComponentPose=@n98, Alpha=1, Translation=Get floatB(), Rotation="0,0,0", Scale="1,1,1") {
    rotationMode=BMM_Additive
    rotationSpace=BCS_BoneSpace
    scaleMode=BMM_Additive
    scaleSpace=BCS_BoneSpace
    translationMode=BMM_Additive
    translationSpace=BCS_BoneSpace
    alphaBoolBlend {blendOption=Linear}
    boneToModify {boneName=dyn_shard_knee_a_l}
   }
   n100 = "Transform (Modify) Bone - Bone: dyn_shard_torso_a_l"(ComponentPose=@n99, Alpha=0.7, Translation=Get floatA(), Rotation="0,0,0", Scale="1,1,1") {
    rotationMode=BMM_Additive
    rotationSpace=BCS_BoneSpace
    scaleMode=BMM_Additive
    scaleSpace=BCS_BoneSpace
    translationMode=BMM_Additive
    translationSpace=BCS_BoneSpace
    alphaBoolBlend {blendOption=Linear}
    boneToModify {boneName=dyn_shard_torso_a_l}
   }
   n101 = "Transform (Modify) Bone - Bone: dyn_shard_shoulder_a_l"(ComponentPose=@n100, Alpha=0.5, Translation=Get floatA(), Rotation="0,0,0", Scale="1,1,1") {
    rotationMode=BMM_Additive
    rotationSpace=BCS_BoneSpace
    scaleMode=BMM_Additive
    scaleSpace=BCS_BoneSpace
    translationMode=BMM_Additive
    translationSpace=BCS_BoneSpace
    alphaBoolBlend {blendOption=Linear}
    boneToModify {boneName=dyn_shard_shoulder_a_l}
   }
   n102 = "Transform (Modify) Bone - Bone: dyn_shard_shoulder_b_l"(ComponentPose=@n101, Alpha=0.7, Translation=Get floatB(), Rotation="0,0,0", Scale="1,1,1") {
    rotationMode=BMM_Additive
    rotationSpace=BCS_BoneSpace
    scaleMode=BMM_Additive
    scaleSpace=BCS_BoneSpace
    translationMode=BMM_Additive
    translationSpace=BCS_BoneSpace
    alphaBoolBlend {blendOption=Linear}
    boneToModify {boneName=dyn_shard_shoulder_b_l}
   }
   n103 = "Transform (Modify) Bone - Bone: dyn_shard_shoulder_c_l"(ComponentPose=@n102, Alpha=1, Translation=Get floatA(), Rotation="0,0,0", Scale="1,1,1") {
    rotationMode=BMM_Additive
    rotationSpace=BCS_BoneSpace
    scaleMode=BMM_Additive
    scaleSpace=BCS_BoneSpace
    translationMode=BMM_Additive
    translationSpace=BCS_BoneSpace
    alphaBoolBlend {blendOption=Linear}
    boneToModify {boneName=dyn_shard_shoulder_c_l}
   }
   n104 = "Transform (Modify) Bone - Bone: dyn_shard_shoulder_d_l"(ComponentPose=@n103, Alpha=1, Translation=Get floatB(), Rotation="0,0,0", Scale="1,1,1") {
    rotationMode=BMM_Additive
    rotationSpace=BCS_BoneSpace
    scaleMode=BMM_Additive
    scaleSpace=BCS_BoneSpace
    translationMode=BMM_Additive
    translationSpace=BCS_BoneSpace
    alphaBoolBlend {blendOption=Linear}
    boneToModify {boneName=dyn_shard_shoulder_d_l}
   }
   n105 = "Transform (Modify) Bone - Bone: dyn_shard_shoulder_e_l"(ComponentPose=@n104, Alpha=1, Translation=Get floatA(), Rotation="0,0,0", Scale="1,1,1") {
    rotationMode=BMM_Additive
    rotationSpace=BCS_BoneSpace
    scaleMode=BMM_Additive
    scaleSpace=BCS_BoneSpace
    translationMode=BMM_Additive
    translationSpace=BCS_BoneSpace
    alphaBoolBlend {blendOption=Linear}
    boneToModify {boneName=dyn_shard_shoulder_e_l}
   }
   n106 = "Transform (Modify) Bone - Bone: dyn_shard_shoulder_f_l"(ComponentPose=@n105, Alpha=1, Translation=Get floatB(), Rotation="0,0,0", Scale="1,1,1") {
    rotationMode=BMM_Additive
    rotationSpace=BCS_BoneSpace
    scaleMode=BMM_Additive
    scaleSpace=BCS_BoneSpace
    translationMode=BMM_Additive
    translationSpace=BCS_BoneSpace
    alphaBoolBlend {blendOption=Linear}
    boneToModify {boneName=dyn_shard_shoulder_f_l}
   }
   n107 = "Transform (Modify) Bone - Bone: dyn_shard_shoulder_g_l"(ComponentPose=@n106, Alpha=0.5, Translation=Get floatA(), Rotation="0,0,0", Scale="1,1,1") {
    rotationMode=BMM_Additive
    rotationSpace=BCS_BoneSpace
    scaleMode=BMM_Additive
    scaleSpace=BCS_BoneSpace
    translationMode=BMM_Additive
    translationSpace=BCS_BoneSpace
    alphaBoolBlend {blendOption=Linear}
    boneToModify {boneName=dyn_shard_shoulder_g_l}
   }
   n108 = "Transform (Modify) Bone - Bone: dyn_shard_shoulder_h_l"(ComponentPose=@n107, Alpha=1, Translation=Get floatB(), Rotation="0,0,0", Scale="1,1,1") {
    rotationMode=BMM_Additive
    rotationSpace=BCS_BoneSpace
    scaleMode=BMM_Additive
    scaleSpace=BCS_BoneSpace
    translationMode=BMM_Additive
    translationSpace=BCS_BoneSpace
    alphaBoolBlend {blendOption=Linear}
    boneToModify {boneName=dyn_shard_shoulder_h_l}
   }
   Save cached pose 'CrystalCachedPose'(Pose=Component To Local(ComponentPose=@n108)) {cacheName=CrystalCachedPose}
   n109 = "Constraint - dyn_skirt_fr_a_base_l (RR)"!(
    ComponentPose=Local To Component(LocalPose=Use cached pose 'deformRigPose'())
    Alpha=1
    ConstraintWeights_0=0.1
    ConstraintWeights_1:float=0.9
   ) -> (Pose) {
    alphaBoolBlend {blendOption=Linear}
    boneToModify {boneName=dyn_skirt_fr_a_base_l}
    constraintSetup [{transformType=Rotation, targetBone {boneName=pelvis}}, {transformType=Rotation, targetBone {boneName=thigh_l}}]
    constraintWeights [1,1]
   }
   n110 = "Constraint - dyn_skirt_side_a_base_l (RR)"!(ComponentPose=@n109, Alpha=1, ConstraintWeights_0=0.4, ConstraintWeights_1:float=0.6) -> (Pose) {
    alphaBoolBlend {blendOption=Linear}
    boneToModify {boneName=dyn_skirt_side_a_base_l}
    constraintSetup [{transformType=Rotation, targetBone {boneName=pelvis}}, {transformType=Rotation, targetBone {boneName=thigh_l}}]
    constraintWeights [1,1]
   }
   n111 = "Constraint - dyn_skirt_bk_a_base_l (RR)"!(ComponentPose=@n110, Alpha=1, ConstraintWeights_0=0.8, ConstraintWeights_1:float=0.2) -> (Pose) {
    alphaBoolBlend {blendOption=Linear}
    boneToModify {boneName=dyn_skirt_bk_a_base_l}
    constraintSetup [{transformType=Rotation, targetBone {boneName=pelvis}}, {transformType=Rotation, targetBone {boneName=thigh_l}}]
    constraintWeights [1,1]
   }
   n112 = "Constraint - dyn_skirt_fr_a_base_r (RR)"!(ComponentPose=@n111, Alpha=1, ConstraintWeights_0=0.1, ConstraintWeights_1:float=0.9) -> (Pose) {
    alphaBoolBlend {blendOption=Linear}
    boneToModify {boneName=dyn_skirt_fr_a_base_r}
    constraintSetup [{transformType=Rotation, targetBone {boneName=pelvis}}, {transformType=Rotation, targetBone {boneName=thigh_r}}]
    constraintWeights [1,1]
   }
   n113 = "Constraint - dyn_skirt_side_a_base_r (RR)"!(ComponentPose=@n112, Alpha=1, ConstraintWeights_0=0.4, ConstraintWeights_1:float=0.6) -> (Pose) {
    alphaBoolBlend {blendOption=Linear}
    boneToModify {boneName=dyn_skirt_side_a_base_r}
    constraintSetup [{transformType=Rotation, targetBone {boneName=pelvis}}, {transformType=Rotation, targetBone {boneName=thigh_r}}]
    constraintWeights [1,1]
   }
   n114 = "Constraint - dyn_skirt_bk_a_base_r (RR)"!(ComponentPose=@n113, Alpha=1, ConstraintWeights_0=0.8, ConstraintWeights_1:float=0.2) -> (Pose) {
    alphaBoolBlend {blendOption=Linear}
    boneToModify {boneName=dyn_skirt_bk_a_base_r}
    constraintSetup [{transformType=Rotation, targetBone {boneName=pelvis}}, {transformType=Rotation, targetBone {boneName=thigh_r}}]
    constraintWeights [1,1]
   }
   n115 = Break Buffet Pose Space Deformer Input(BuffetPoseSpaceDeformerInput=Get ThighRotL()) -> (DeformationValue=~)
   n116 = "Transform (Modify) Bone - Bone: dyn_skirt_fr_a_base_l"(ComponentPose=@n114, Alpha=@n115, Translation="2,0,3", Rotation="0,0,0", Scale="1,1,1") {
    rotationMode=BMM_Additive
    rotationSpace=BCS_BoneSpace
    scaleMode=BMM_Additive
    scaleSpace=BCS_BoneSpace
    translationMode=BMM_Additive
    translationSpace=BCS_BoneSpace
    alphaBoolBlend {blendOption=Linear}
    alphaScaleBiasClamp {bMapRange=true, inRange {max=0, min=-100}, outRange {max=0, min=1}}
    boneToModify {boneName=dyn_skirt_fr_a_base_l}
    translation {x=2, z=3}
   }
   n117 = "Transform (Modify) Bone - Bone: dyn_skirt_side_a_base_l"(ComponentPose=@n116, Alpha=@n115, Translation="0,0,4", Rotation="0,0,0", Scale="1,1,1") {
    rotationMode=BMM_Additive
    rotationSpace=BCS_BoneSpace
    scaleMode=BMM_Additive
    scaleSpace=BCS_BoneSpace
    translationMode=BMM_Additive
    translationSpace=BCS_BoneSpace
    alphaBoolBlend {blendOption=Linear}
    alphaScaleBiasClamp {bMapRange=true, inRange {max=0, min=-100}, outRange {max=0, min=1}}
    boneToModify {boneName=dyn_skirt_side_a_base_l}
    translation {z=4}
   }
   n118 = Break Buffet Pose Space Deformer Input(BuffetPoseSpaceDeformerInput=Get ThighRotR()) -> (DeformationValue=~)
   n119 = "Transform (Modify) Bone - Bone: dyn_skirt_fr_a_base_r"(ComponentPose=@n117, Alpha=@n118, Translation="-2,0,-3", Rotation="0,0,0", Scale="1,1,1") {
    rotationMode=BMM_Additive
    rotationSpace=BCS_BoneSpace
    scaleMode=BMM_Additive
    scaleSpace=BCS_BoneSpace
    translationMode=BMM_Additive
    translationSpace=BCS_BoneSpace
    alphaBoolBlend {blendOption=Linear}
    alphaScaleBiasClamp {bMapRange=true, inRange {max=0, min=-100}, outRange {max=0, min=1}}
    boneToModify {boneName=dyn_skirt_fr_a_base_r}
    translation {x=-2, z=-3}
   }
   n120 = "Transform (Modify) Bone - Bone: dyn_skirt_side_a_base_r"(ComponentPose=@n119, Alpha=@n118, Translation="0,0,-4", Rotation="0,0,0", Scale="1,1,1") {
    rotationMode=BMM_Additive
    rotationSpace=BCS_BoneSpace
    scaleMode=BMM_Additive
    scaleSpace=BCS_BoneSpace
    translationMode=BMM_Additive
    translationSpace=BCS_BoneSpace
    alphaBoolBlend {blendOption=Linear}
    alphaScaleBiasClamp {bMapRange=true, inRange {max=0, min=-100}, outRange {max=0, min=1}}
    boneToModify {boneName=dyn_skirt_side_a_base_r}
    translation {z=-4}
   }
   n121 = RigidBody!(
    ComponentPose=@n120
    Alpha=Get crouchingAlpha() -> (crouchingAlpha:float)
    OverrideWorldGravity:Vector=Get gravityOverride() -> (gravityOverride:Vector)
    ExternalForce
   ) -> (Pose) {
    bClampLinearTranslationLimitToRefPose=true
    bOverrideWorldGravity=true
    overridePhysicsAsset=&/F_MED_Buffet_Shard_Physics2
    simulationSpace=BaseBoneSpace
    alphaBoolBlend {blendOption=Linear}
    alphaScaleBiasClamp {bMapRange=true, outRange {max=0.5, min=1}}
    baseBoneRef {boneName=simspace}
    componentAppliedLinearAccClamp=<3000 3000 3000>
    componentLinearAccScale=<0.25 0.25 0.25>
    componentLinearVelScale=<0.25 0.25 0.25>
    overrideWorldGravity {z=-980}
   }
   n122 = "Constraint - dyn_skirt_fr_a_base (PPP)"!(
    ComponentPose=@n121
    Alpha=Get skirtMidJntConstraintAlpha() -> (skirtMidJntConstraintAlpha:float)
    ConstraintWeights_0=0.5
    ConstraintWeights_1:float=0.5
    ConstraintWeights_2:float=0
   ) -> (Pose) {
    alphaBoolBlend {blendOption=Linear}
    alphaScaleBiasClamp {bClampResult=true, bInterpResult=true, clampMax=0.3, interpSpeedDecreasing=5, interpSpeedIncreasing=12}
    boneToModify {boneName=dyn_skirt_fr_a_base}
    constraintSetup [{targetBone {boneName=dyn_skirt_fr_a_l}}, {targetBone {boneName=dyn_skirt_fr_a_r}}, {targetBone {boneName=pelvis}}]
    constraintWeights [1,1,1]
   }
   n123 = "Constraint - dyn_skirt_bk_a_base (PPP)"!(ComponentPose=@n122, Alpha=1, ConstraintWeights_0=0.5, ConstraintWeights_1:float=0.5, ConstraintWeights_2:float=0) -> (Pose) {
    alphaBoolBlend {blendOption=Linear}
    boneToModify {boneName=dyn_skirt_bk_a_base}
    constraintSetup [{targetBone {boneName=dyn_skirt_bk_a_l}}, {targetBone {boneName=dyn_skirt_bk_a_r}}, {targetBone {boneName=pelvis}}]
    constraintWeights [1,1,1]
   }
   Save cached pose 'SkirtCachePose'(Pose=Component To Local(ComponentPose=@n123)) {cacheName=SkirtCachePose}
  ]}
 EventGraph {nodes [
   Event BlueprintUpdateAnimation()
   > n1 = ComplexSimSpaceFunction DEPRECATED!(
    execute
    CharacterPartAnimInstance:Custom Character Part Anim Instance Object Reference=Self-Reference()
    InterpSpeedMax:double=9
    PawnSpeedMax:double=20
    RInterpErrorTolerance:double=1
    DeltaTime:double=0
    DebugDraw:bool=false
    sim_space_rotation_in:Rotator=Get sim_space_rotation_$()
    inComponentSpaceIn:bool=Get InComponentSpace_$() -> (InComponentSpace_$:bool)
    inComponentSpaceLastFrameIn:bool=Get InComponentSpaceLastFrame_$() -> (InComponentSpaceLastFrame_$:bool)
    lockedOnSimSpaceIn:bool=Get LockedOnSimSpace_$() -> (LockedOnSimSpace_$:bool)
    InterpSpeedIn:double=Get InterpSpeed_$() -> (InterpSpeed_$:double)
   ) -> (
    then
    sim_space_rotation:Rotator
    inComponentSpaceOut:bool
    inComponentSpaceLastFrameOut:bool
    lockedOnSimSpaceOut:bool
    InterpSpeedOut:double
   )
   > Set sim_space_rotation_$(sim_space_rotation_$:Rotator=@n1) -> (Output_Get:Rotator)
   > Set InComponentSpace_$(InComponentSpace_$:bool=@n1.inComponentSpaceOut) -> (Output_Get:bool)
   > Set InComponentSpaceLastFrame_$(InComponentSpaceLastFrame_$:bool=@n1.inComponentSpaceLastFrameOut) -> (Output_Get:bool)
   > Set LockedOnSimSpace_$(LockedOnSimSpace_$:bool=@n1.lockedOnSimSpaceOut) -> (Output_Get:bool)
   > Set InterpSpeed_$(InterpSpeed_$:double=@n1.InterpSpeedOut) -> (Output_Get:double)
  ]}
]
variables [
 Body_Mesh {type=SkeletalMeshComponent}
 InComponentSpace_$ {type=boolean}
 InComponentSpaceLastFrame_$ {type=boolean}
 InterpSpeed_$ {type=number}
 LockedOnSimSpace_$ {type=boolean}
 sim_space_rotation_$ {type=Rotator}
]
